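; Parent template for melee manager variants that turns per-category user options into tuned melee weapon stats for valid item owners.
; ApplyOverrides writes options into the by-ref stats struct when a category is enabled or forced.
; CombineModifiers sums enabled add/mult options into a modifier struct.
; ApplyModifiers feeds that struct through per-category functions (weapon, attacks, combo, block/dodge, glow) using CalcModifier, with combo hits keyed by gameplay tag.
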(fn ApplyModifierBasicAttack
  (params (StateValues :type Creative%20Melee%20Options%20Structure%20%28by%20ref%29) (StateModifier :type Creative%20Melee%20Modifier%20Options%20Structure))
  (bind (_knockbackadd_29_dda919654e1b8959a586b7a358284ce3 _knockbackmult_30_e1d934af4c63176e2f905cbefd4d56aa _knockbackplayeradd_31_46cfdb8f4941e642043b2983d5c385f5 _knockbackaiadd_32_d8dc593a4b40a28c7277c9a947b7220b) (Utilities|Struct|BreakCreativeMeleeModifierOptions (Variables|GetStateModifier)))
  (bind (_knockback_20_288c1de84fb83660a91876b19b39ee3b _aiknockback_86_3c11a6804d38dce230c48086a05545b1) (Utilities|Struct|BreakCreativeMeleeOptions (LocalVariables|GetmyValues)))
  (bind (_damageadd_2_7cec514c49df151a9d453491b01e4f93 _damagemult_20_200dd6cd436b4c20577174ac49310020 _damageplayeradd_21_494fd97541d4f0520f854fb319b2e853 _damageplayermult_22_f168716742252fc272e9aaae92d38cee _damageaiadd_23_b7d849c743710fa3e555a18c0d32ffbe _damageaimult_24_bd1f0cb248c75414f1906fac3726a0c8 _damageenvrionmentadd_25_f16ebbf64c5446d214efee9f5c84683d _damageenvrionmentmult_26_4882de4945cc65d5995fe58d85b2e1b1) (Utilities|Struct|BreakCreativeMeleeModifierOptions (Variables|GetStateModifier)))
  (bind (_damage_142_2efb279a4a3f30f80bc9dd86a992cba6 _basedamage_16_b01234df44393696f04c92b8d351e18e _baseaidamage_92_50312ec54dca1e86ae32918d5be84b5d _environmentdamage_28_0a4087964fb24d1fcb33038d472ce9cb) (Utilities|Struct|BreakCreativeMeleeOptions (LocalVariables|GetmyValues)))
  (bind (_energycostadd_105_326d529b46e858d45447c99058290b4d _energycostmult_106_42c56f8943098c5e7126fa8a1ff3c567 _comboenergycostadd_132_c87ec0354eb67cc449cb25813480bea8) (Utilities|Struct|BreakCreativeMeleeModifierOptions (Variables|GetStateModifier)))
  (LocalVariables|SetmyValues StateValues)
  (Utilities|FlowControl|Sequence
    (:then_0
      ;[1 1 1] Damage
      (Utilities|Struct|SetmembersinCreativeMeleeOptions (LocalVariables|GetmyValues) (Utilities|CalcModifier (+ _damage_142_2efb279a4a3f30f80bc9dd86a992cba6 _basedamage_16_b01234df44393696f04c92b8d351e18e) (+ _damageadd_2_7cec514c49df151a9d453491b01e4f93 _damageplayeradd_21_494fd97541d4f0520f854fb319b2e853) (+ _damagemult_20_200dd6cd436b4c20577174ac49310020 _damageplayermult_22_f168716742252fc272e9aaae92d38cee)) (Utilities|CalcModifier (+ _damage_142_2efb279a4a3f30f80bc9dd86a992cba6 _baseaidamage_92_50312ec54dca1e86ae32918d5be84b5d) (+ _damageadd_2_7cec514c49df151a9d453491b01e4f93 _damageaiadd_23_b7d849c743710fa3e555a18c0d32ffbe) (+ _damagemult_20_200dd6cd436b4c20577174ac49310020 _damageaimult_24_bd1f0cb248c75414f1906fac3726a0c8)) (Utilities|CalcModifier (+ _damage_142_2efb279a4a3f30f80bc9dd86a992cba6 _environmentdamage_28_0a4087964fb24d1fcb33038d472ce9cb) (+ _damageadd_2_7cec514c49df151a9d453491b01e4f93 _damageenvrionmentadd_25_f16ebbf64c5446d214efee9f5c84683d) (+ _damagemult_20_200dd6cd436b4c20577174ac49310020 _damageenvrionmentmult_26_4882de4945cc65d5995fe58d85b2e1b1)))
      ;[1 1 1] Knockback
      (Utilities|Struct|SetmembersinCreativeMeleeOptions (LocalVariables|GetmyValues) (Utilities|CalcModifier _knockback_20_288c1de84fb83660a91876b19b39ee3b (+ _knockbackadd_29_dda919654e1b8959a586b7a358284ce3 _knockbackplayeradd_31_46cfdb8f4941e642043b2983d5c385f5) _knockbackmult_30_e1d934af4c63176e2f905cbefd4d56aa) (Utilities|CalcModifier _aiknockback_86_3c11a6804d38dce230c48086a05545b1 (+ _knockbackadd_29_dda919654e1b8959a586b7a358284ce3 _knockbackaiadd_32_d8dc593a4b40a28c7277c9a947b7220b) _knockbackmult_30_e1d934af4c63176e2f905cbefd4d56aa))
      ;[1 1 1] Energy Cost
      (Utilities|Struct|SetmembersinCreativeMeleeOptions (LocalVariables|GetmyValues) (Utilities|CalcModifier (Utilities|Struct|BreakCreativeMeleeOptions (LocalVariables|GetmyValues)) (+ _energycostadd_105_326d529b46e858d45447c99058290b4d _comboenergycostadd_132_c87ec0354eb67cc449cb25813480bea8) _energycostmult_106_42c56f8943098c5e7126fa8a1ff3c567)))
    (:then_1
      (Variables|SetBy-RefVar StateValues (LocalVariables|GetmyValues)))))

(fn ApplyModifierChargeAttack
  (params (StateValues :type Creative%20Melee%20Options%20Structure%20%28by%20ref%29) (StateModifier :type Creative%20Melee%20Modifier%20Options%20Structure))
  (bind (_damageadd_2_7cec514c49df151a9d453491b01e4f93 _damagemult_20_200dd6cd436b4c20577174ac49310020 _knockbackadd_29_dda919654e1b8959a586b7a358284ce3 _knockbackmult_30_e1d934af4c63176e2f905cbefd4d56aa _chargeattackdamageadd_52_3176a2b14f91c9b93557ad9fdb6c70e3 _chargeattackdamagemult_53_435d848547cbd97806a7f981762cb4a9 _chargeattackknockback_54_142d815243467a83e732089c62daba27 _chargeattackcost_55_96697e5b4a48cf875d2ea08c1b1c5553 _chargetimeadd_56_18fa8ab14ea7f17605b1a694ec189e6e _chargetimemult_57_ad55b88343ccc3719ebb66b9420742e0) (Utilities|Struct|BreakCreativeMeleeModifierOptions (Variables|GetStateModifier)))
  (bind (_secondarydamage_32_eb6925f04b8c5ca1ea314abbc9c1b68f _secondaryknockback_47_20879ce6494d1742b4c73d8863396a13 _totalchargetime_51_158e8c1d4ad3710b7babdfb2edab874f _secondaryenergycost_100_1556aeee4d244b5f2bc7c0ba9ea6b0d4) (Utilities|Struct|BreakCreativeMeleeOptions (LocalVariables|GetmyValues)))
  (LocalVariables|SetmyValues StateValues)
  (Utilities|FlowControl|Sequence
    (:then_0
      ;[1 1 1] Charge Attack properties
      (Utilities|Struct|SetmembersinCreativeMeleeOptions (LocalVariables|GetmyValues) (Utilities|CalcModifier _secondarydamage_32_eb6925f04b8c5ca1ea314abbc9c1b68f (+ _damageadd_2_7cec514c49df151a9d453491b01e4f93 _chargeattackdamageadd_52_3176a2b14f91c9b93557ad9fdb6c70e3) (+ _damagemult_20_200dd6cd436b4c20577174ac49310020 _chargeattackdamagemult_53_435d848547cbd97806a7f981762cb4a9)) (Utilities|CalcModifier _secondaryknockback_47_20879ce6494d1742b4c73d8863396a13 (+ _knockbackadd_29_dda919654e1b8959a586b7a358284ce3 _chargeattackknockback_54_142d815243467a83e732089c62daba27) _knockbackmult_30_e1d934af4c63176e2f905cbefd4d56aa) (Utilities|CalcModifier _totalchargetime_51_158e8c1d4ad3710b7babdfb2edab874f _chargetimeadd_56_18fa8ab14ea7f17605b1a694ec189e6e _chargetimemult_57_ad55b88343ccc3719ebb66b9420742e0) (Utilities|CalcModifier _secondaryenergycost_100_1556aeee4d244b5f2bc7c0ba9ea6b0d4 _chargeattackcost_55_96697e5b4a48cf875d2ea08c1b1c5553)))
    (:then_1
      (Variables|SetBy-RefVar StateValues (LocalVariables|GetmyValues)))))

(fn ApplyModifierCombo
  (params (StateValues :type Creative%20Melee%20Options%20Structure%20%28by%20ref%29) (StateModifier :type Creative%20Melee%20Modifier%20Options%20Structure))
  (LocalVariables|SetmyValues StateValues)
  (Utilities|FlowControl|Sequence
    (:then_0
      ;[1 1 1] Attack Chain
      (bind _keys (Utilities|Map|Keys (Utilities|Struct|BreakCreativeMeleeOptions (LocalVariables|GetmyValues))))
      (for _array_element _keys
        (Utilities|Map|Add (LocalVariables|GetmyAttacks) _array_element (Utilities|Struct|MakeCreativeMeleeHitOptions (+ (Combo|GetComboValues _array_element) (.DamageMult (Combo|GetComboModValues _array_element (Variables|GetStateModifier)))) (+ (.KnockbackMult (Combo|GetComboValues _array_element)) (.KnockbackMult (Combo|GetComboModValues _array_element (Variables|GetStateModifier)))) (+ (.EnergyMult (Combo|GetComboValues _array_element)) (.EnergyMult (Combo|GetComboModValues _array_element (Variables|GetStateModifier)))))))
      (Utilities|Struct|SetmembersinCreativeMeleeOptions (LocalVariables|GetmyValues) (LocalVariables|GetmyAttacks)))
    (:then_1
      (Variables|SetBy-RefVar StateValues (LocalVariables|GetmyValues)))))

(fn ApplyModifierJumpAttack
  (params (StateValues :type Creative%20Melee%20Options%20Structure%20%28by%20ref%29) (StateModifier :type Creative%20Melee%20Modifier%20Options%20Structure))
  (bind (_damageadd_2_7cec514c49df151a9d453491b01e4f93 _damagemult_20_200dd6cd436b4c20577174ac49310020 _knockbackadd_29_dda919654e1b8959a586b7a358284ce3 _knockbackmult_30_e1d934af4c63176e2f905cbefd4d56aa _energycostadd_105_326d529b46e858d45447c99058290b4d _energycostmult_106_42c56f8943098c5e7126fa8a1ff3c567 _jumpattackdamageadd_108_74fc14134de8d68c8796f6a2ac9a0f40 _jumpattackdamagemult_109_e14a07824facae9812220da7dd827275 _jumpattackknockback_110_c62835bd45a7b5ea56e62b970de4b240 _jumpattackcost_111_1ff24cc649aefc9c0427088db2eaa980) (Utilities|Struct|BreakCreativeMeleeModifierOptions (Variables|GetStateModifier)))
  (bind (_jumpattackdamage_63_e597e76d4747bcad2a2d05aa4abc4f9e _jumpattackknockbackpower_65_446e41444a48e9bef6552c85cab864cb _jumpattackenergycost_129_dea941b7458e0ad503a8e5b69cae6527) (Utilities|Struct|BreakCreativeMeleeOptions (LocalVariables|GetmyValues)))
  (LocalVariables|SetmyValues StateValues)
  (Utilities|FlowControl|Sequence
    (:then_0
      ;[1 1 1] Sprint Attack properties
      (Utilities|Struct|SetmembersinCreativeMeleeOptions (LocalVariables|GetmyValues) (Utilities|CalcModifier _jumpattackdamage_63_e597e76d4747bcad2a2d05aa4abc4f9e (+ _damageadd_2_7cec514c49df151a9d453491b01e4f93 _jumpattackdamageadd_108_74fc14134de8d68c8796f6a2ac9a0f40) (+ _damagemult_20_200dd6cd436b4c20577174ac49310020 _jumpattackdamagemult_109_e14a07824facae9812220da7dd827275)) (Utilities|CalcModifier _jumpattackknockbackpower_65_446e41444a48e9bef6552c85cab864cb (+ _knockbackadd_29_dda919654e1b8959a586b7a358284ce3 _jumpattackknockback_110_c62835bd45a7b5ea56e62b970de4b240) _knockbackmult_30_e1d934af4c63176e2f905cbefd4d56aa) (Utilities|CalcModifier _jumpattackenergycost_129_dea941b7458e0ad503a8e5b69cae6527 (+ _energycostadd_105_326d529b46e858d45447c99058290b4d _jumpattackcost_111_1ff24cc649aefc9c0427088db2eaa980) _energycostmult_106_42c56f8943098c5e7126fa8a1ff3c567)))
    (:then_1
      (Variables|SetBy-RefVar StateValues (LocalVariables|GetmyValues)))))

(fn ApplyModifierPresentation
  (params (StateValues :type Creative%20Melee%20Options%20Structure%20%28by%20ref%29) (StateModifier :type Creative%20Melee%20Modifier%20Options%20Structure))
  (bind (_colorintensitycustom_122_dbc0ae3c46c05c8bbd194bb6f5b27f2d _colorintensityweapon_121_00ed43c2488af7d4c35e0da305ed7695) (Utilities|Struct|BreakCreativeMeleeModifierOptions (Variables|GetStateModifier)))
  (bind (_customvfxintensity_122_e62480714f5bf3dab0a41298a5e42a6e _weaponglowintensity_123_72d787f14737f1b1cb073aa54af7554c) (Utilities|Struct|BreakCreativeMeleeOptions (LocalVariables|GetmyValues)))
  (LocalVariables|SetmyValues StateValues)
  (Utilities|FlowControl|Sequence
    (:then_0
      ;[1 1 1] Weapon Glow
      (Utilities|Struct|SetmembersinCreativeMeleeOptions (LocalVariables|GetmyValues) (Utilities|CalcModifier _customvfxintensity_122_e62480714f5bf3dab0a41298a5e42a6e _colorintensitycustom_122_dbc0ae3c46c05c8bbd194bb6f5b27f2d) (Utilities|CalcModifier _weaponglowintensity_123_72d787f14737f1b1cb073aa54af7554c _colorintensityweapon_121_00ed43c2488af7d4c35e0da305ed7695)))
    (:then_1
      (Variables|SetBy-RefVar StateValues (LocalVariables|GetmyValues)))))

(fn ApplyModifiers
  (params (StateValues :type Creative%20Melee%20Options%20Structure%20%28by%20ref%29) (StateModifier :type Creative%20Melee%20Modifier%20Options%20Structure))
  (LocalVariables|SetmyValues StateValues)
  (Utilities|FlowControl|Sequence
    (:then_0
      (ModifierLogic|ApplyModifierWeapon (LocalVariables|GetmyValues) (Variables|GetStateModifier))
      (ModifierLogic|ApplyModifierBasicAttack (LocalVariables|GetmyValues) (Variables|GetStateModifier))
      (ModifierLogic|ApplyModifierCombo (LocalVariables|GetmyValues) (Variables|GetStateModifier))
      (ModifierLogic|ApplyModifierChargeAttack (LocalVariables|GetmyValues) (Variables|GetStateModifier)) ;   
      (ModifierLogic|ApplyModifierSecondaryAction (LocalVariables|GetmyValues) (Variables|GetStateModifier))
      (ModifierLogic|ApplyModifierJumpAttack (LocalVariables|GetmyValues) (Variables|GetStateModifier))
      (ModifierLogic|ApplyModifierSprintAttack (LocalVariables|GetmyValues) (Variables|GetStateModifier))
      (ModifierLogic|ApplyModifierPresentation (LocalVariables|GetmyValues) (Variables|GetStateModifier)))
    (:then_1
      (Variables|SetBy-RefVar StateValues (LocalVariables|GetmyValues)))))

(fn ApplyModifierSecondaryAction
  (params (StateValues :type Creative%20Melee%20Options%20Structure%20%28by%20ref%29) (StateModifier :type Creative%20Melee%20Modifier%20Options%20Structure))
  (bind (_blockresistadd_101_465abb514c1610b643f8ffb81354e78b _blockresistmult_102_168b5d64491f00141ba2be943655deb6 _blockcostadd_128_0e7a7cdd469ea615ec2863910cb9d664 _blockcostmult_129_afdbf000445b18a6309c148379dcd41a _dodgecostadd_103_bc2ced9743d7fcda19c3449327b41883 _dodgecostmult_104_b13cd3214bc139f9b64610906e25a657 _energycostadd_105_326d529b46e858d45447c99058290b4d _energycostmult_106_42c56f8943098c5e7126fa8a1ff3c567) (Utilities|Struct|BreakCreativeMeleeModifierOptions (Variables|GetStateModifier)))
  (bind (_damageresistance_95_77fe74874590cc7b0bad9da0361692be _blockhitenergycost_102_7652559547687830a0a870a6c1c2e754 _dodgeenergycost_104_b87c80884c0ce9a1f652ddb29205a0c6) (Utilities|Struct|BreakCreativeMeleeOptions (LocalVariables|GetmyValues)))
  (LocalVariables|SetmyValues StateValues)
  (Utilities|FlowControl|Sequence
    (:then_0
      ;[1 1 1] Secondary Action (Guard & Dodge)
      (Utilities|Struct|SetmembersinCreativeMeleeOptions (LocalVariables|GetmyValues) (Utilities|CalcModifier _damageresistance_95_77fe74874590cc7b0bad9da0361692be _blockresistadd_101_465abb514c1610b643f8ffb81354e78b _blockresistmult_102_168b5d64491f00141ba2be943655deb6) (Utilities|CalcModifier _blockhitenergycost_102_7652559547687830a0a870a6c1c2e754 (+ _blockcostadd_128_0e7a7cdd469ea615ec2863910cb9d664 _energycostadd_105_326d529b46e858d45447c99058290b4d) (+ _blockcostmult_129_afdbf000445b18a6309c148379dcd41a _energycostmult_106_42c56f8943098c5e7126fa8a1ff3c567)) (Utilities|CalcModifier _dodgeenergycost_104_b87c80884c0ce9a1f652ddb29205a0c6 (+ _dodgecostadd_103_bc2ced9743d7fcda19c3449327b41883 _energycostadd_105_326d529b46e858d45447c99058290b4d) (+ _dodgecostmult_104_b13cd3214bc139f9b64610906e25a657 _energycostmult_106_42c56f8943098c5e7126fa8a1ff3c567))))
    (:then_1
      (Variables|SetBy-RefVar StateValues (LocalVariables|GetmyValues)))))

(fn ApplyModifierSprintAttack
  (params (StateValues :type Creative%20Melee%20Options%20Structure%20%28by%20ref%29) (StateModifier :type Creative%20Melee%20Modifier%20Options%20Structure))
  (bind (_damageadd_2_7cec514c49df151a9d453491b01e4f93 _damagemult_20_200dd6cd436b4c20577174ac49310020 _knockbackadd_29_dda919654e1b8959a586b7a358284ce3 _knockbackmult_30_e1d934af4c63176e2f905cbefd4d56aa _energycostadd_105_326d529b46e858d45447c99058290b4d _energycostmult_106_42c56f8943098c5e7126fa8a1ff3c567 _sprintattackdamageadd_112_f2f36e9444d861b3bd3b85a4965e9113 _sprintattackdamagemult_113_9743c66b438df4377b1aa48a895fc2a4 _sprintattackknockback_117_1271f5294dbf26c9dfd55299ddcda8f5 _sprintattackcost_118_0d2b902b423227dc68ea20aa53b074ce) (Utilities|Struct|BreakCreativeMeleeModifierOptions (Variables|GetStateModifier)))
  (bind (_sprintattackdamage_110_3585d12b48db48a3b455de9fefaecf97 _sprintattackknockback_132_04a113cc4f5780ef7198fbab20c9d342 _sprintattackenergycost_126_012485474b335a7b14c68f80be65f4a8) (Utilities|Struct|BreakCreativeMeleeOptions (LocalVariables|GetmyValues)))
  (LocalVariables|SetmyValues StateValues)
  (Utilities|FlowControl|Sequence
    (:then_0
      ;[1 1 1] Sprint Attack properties
      (Utilities|Struct|SetmembersinCreativeMeleeOptions (LocalVariables|GetmyValues) (Utilities|CalcModifier _sprintattackdamage_110_3585d12b48db48a3b455de9fefaecf97 (+ _damageadd_2_7cec514c49df151a9d453491b01e4f93 _sprintattackdamageadd_112_f2f36e9444d861b3bd3b85a4965e9113) (+ _damagemult_20_200dd6cd436b4c20577174ac49310020 _sprintattackdamagemult_113_9743c66b438df4377b1aa48a895fc2a4)) (Utilities|CalcModifier _sprintattackknockback_132_04a113cc4f5780ef7198fbab20c9d342 (+ _knockbackadd_29_dda919654e1b8959a586b7a358284ce3 _sprintattackknockback_117_1271f5294dbf26c9dfd55299ddcda8f5) _knockbackmult_30_e1d934af4c63176e2f905cbefd4d56aa) (Utilities|CalcModifier _sprintattackenergycost_126_012485474b335a7b14c68f80be65f4a8 (+ _energycostadd_105_326d529b46e858d45447c99058290b4d _sprintattackcost_118_0d2b902b423227dc68ea20aa53b074ce) _energycostmult_106_42c56f8943098c5e7126fa8a1ff3c567)))
    (:then_1
      (Variables|SetBy-RefVar StateValues (LocalVariables|GetmyValues)))))

(fn ApplyModifierWeapon
  (params (StateValues :type Creative%20Melee%20Options%20Structure%20%28by%20ref%29) (StateModifier :type Creative%20Melee%20Modifier%20Options%20Structure))
  (bind (_critchance_27_1b71a1494c54ac4c7ac7d29caff00aa4 _critdamage_28_66fe207a47f411267a118fa55b30d580 _movementbonus_107_a6536df041780566a4b9bca2fe3d1d2b) (Utilities|Struct|BreakCreativeMeleeModifierOptions (Variables|GetStateModifier)))
  (bind (_critchance_18_b6bd1e7e4d026531de947fa85545974c _critmultiplier_44_3919d87c484c89957bee4b9f19883947 _movementspeed_30_9444e7814d756147dba496bb6ea702c5) (Utilities|Struct|BreakCreativeMeleeOptions (LocalVariables|GetmyValues)))
  (LocalVariables|SetmyValues StateValues)
  (Utilities|FlowControl|Sequence
    (:then_0
      ;[1 1 1] Core Weapon Properties
      (Utilities|Struct|SetmembersinCreativeMeleeOptions (LocalVariables|GetmyValues) (Utilities|CalcModifier _critchance_18_b6bd1e7e4d026531de947fa85545974c _critchance_27_1b71a1494c54ac4c7ac7d29caff00aa4) (Utilities|CalcModifier _critmultiplier_44_3919d87c484c89957bee4b9f19883947 _critdamage_28_66fe207a47f411267a118fa55b30d580) (Utilities|CalcModifier _movementspeed_30_9444e7814d756147dba496bb6ea702c5 _movementbonus_107_a6536df041780566a4b9bca2fe3d1d2b)))
    (:then_1
      (Variables|SetBy-RefVar StateValues (LocalVariables|GetmyValues)))))

(fn ApplyOverrides
  (params (ItemOwner :type Actor%20Object%20Reference) (AccumulatedState :type Creative%20Melee%20Options%20Structure%20%28by%20ref%29) (ForceOverrides :type Boolean))
  (bind _valid (ValidforPlayer|ValidForActor ItemOwner))
  (if _valid
    (LocalVariables|SetmyOverride AccumulatedState)
    (Utilities|FlowControl|Sequence
      (:then_0
        ;[1 1 1] Damage
        (if (or (Variables|GetForceOverrides) (Utilities|Enum|Equal(Enum) (Variables|UserOptions|Damage|GetOverrideDamage) "NewEnumerator2"))
          (Utilities|Struct|SetmembersinCreativeMeleeOptions (LocalVariables|GetmyOverride) (Variables|UserOptions|Damage|GetDamage) (Variables|UserOptions|Damage|GetBaseDamage) (Variables|UserOptions|Damage|GetBaseAIDamage) (Variables|UserOptions|Damage|GetEnvironmentDamage) (Variables|UserOptions|Damage|GetCritChance) (Variables|UserOptions|Damage|GetCritDamageMult) (Variables|UserOptions|Extra|GetMovementBonus) (Variables|UserOptions|Damage|GetKnockback) (Variables|UserOptions|Damage|GetAIKnockback) (Variables|UserOptions|Extra|GetPrimaryEnergyCost) (Variables|UserOptions|Extra|GetNoEnergyDamageMultiplier))
          (else
            _))
        ;[1 1 1] Charge Attacks
        (if (or (Variables|GetForceOverrides) (Utilities|Enum|Equal(Enum) (Variables|UserOptions|ChargeAttack|GetOverrideChargeAttack) "NewEnumerator2"))
          (Utilities|Struct|SetmembersinCreativeMeleeOptions (LocalVariables|GetmyOverride) (Variables|UserOptions|ChargeAttack|GetSecondaryDamage) (Variables|UserOptions|ChargeAttack|GetSecondaryAttack) (Variables|UserOptions|ChargeAttack|GetSecondaryKnockback) (Variables|UserOptions|ChargeAttack|GetChargeTime) (CallFunction|GetCannonballImpactRadius) (Math|Conversions|ToInteger(Boolean) (Variables|UserOptions|GetAllowSecondaryAttackToBreakBlock)) (Variables|UserOptions|GetFeedbackScale) (Variables|UserOptions|ChargeAttack|GetSecondaryEnergyCost))
          (else
            _))
        ;[1 1 1] Secondary / Extra
        (if (or (Variables|GetForceOverrides) (Utilities|Enum|Equal(Enum) (Variables|UserOptions|Extra|GetOverrideSecondaryAction) "NewEnumerator2"))
          (Utilities|Struct|SetmembersinCreativeMeleeOptions (LocalVariables|GetmyOverride) (Variables|UserOptions|Extra|GetMovementBonus) (Variables|UserOptions|Extra|GetTargetActionSelection) (Variables|UserOptions|Extra|GetLockOn) (Variables|UserOptions|Extra|GetDamageResistance) (Variables|UserOptions|Extra|GetBlockHitEnergyCost) (Variables|UserOptions|Extra|GetDodgeEnergyCost) (Variables|UserOptions|Extra|GetNoEnergyDamageMultiplier))
          (else
            _))
        ;[1 1 1] Jump Attacks
        (if (or (Variables|GetForceOverrides) (Utilities|Enum|Equal(Enum) (Variables|UserOptions|JumpAttack|GetOverrideJumpAttack) "NewEnumerator2"))
          (Utilities|Struct|SetmembersinCreativeMeleeOptions (LocalVariables|GetmyOverride) (Variables|UserOptions|JumpAttack|GetJumpAttackEnabled) (Variables|UserOptions|JumpAttack|GetJumpAttackDamage) (Variables|UserOptions|JumpAttack|GetJumpAttackKnockbackPower) (Variables|UserOptions|JumpAttack|GetJumpAttackEnergyCost))
          (else
            _))
        ;[1 1 1] Sprint Attacks
        (if (or (Variables|GetForceOverrides) (Utilities|Enum|Equal(Enum) (Variables|UserOptions|SprintAttack|GetOverrideSprintAttack) "NewEnumerator2"))
          (Utilities|Struct|SetmembersinCreativeMeleeOptions (LocalVariables|GetmyOverride) (Variables|UserOptions|SprintAttack|GetSprintAttackEnabled) (Variables|UserOptions|SprintAttack|GetSprintAttackDamage) (Variables|UserOptions|SprintAttack|GetSprintAttackKnockback) (Variables|UserOptions|SprintAttack|GetSprintAttackEnergyCost))
          (else
            _))
        ;[1 1 1] Presentation - Review UX
        (if (or (Variables|GetForceOverrides) (Utilities|Enum|Equal(Enum) (Variables|UserOptions|Presentation|GetOverridePresentation) "NewEnumerator2"))
          (Utilities|Struct|SetmembersinCreativeMeleeOptions (LocalVariables|GetmyOverride) (Variables|UserOptions|Presentation|GetUseCustomVFXColor) (Variables|UserOptions|Presentation|GetCustomVFXColor) (Variables|UserOptions|Presentation|GetCustomVFXIntensity) (Variables|UserOptions|Presentation|GetWeaponGlow))
          (else
            _))
        ;[1 0.6888 0] Attack Chain  (Revisit per UEFN and Melee Refactor)
        (if (or (Variables|GetForceOverrides) (Utilities|Enum|Equal(Enum) (Variables|UserOptions|Attacks|GetOverrideCombo) "NewEnumerator2"))
          (bind _keys (Utilities|Map|Keys (Utilities|Struct|BreakCreativeMeleeOptions (LocalVariables|GetmyOverride))))
          (for _array_element _keys
            (Utilities|Map|Add (LocalVariables|GetmyAttacks) _array_element (Utilities|Struct|MakeCreativeMeleeHitOptions (Combo|GetComboValues _array_element) (.KnockbackMult (Combo|GetComboValues _array_element)) (.EnergyMult (Combo|GetComboValues _array_element)))))
          (Utilities|Struct|SetmembersinCreativeMeleeOptions (LocalVariables|GetmyOverride) (LocalVariables|GetmyAttacks))))
      (:then_1
        (Variables|SetBy-RefVar AccumulatedState (LocalVariables|GetmyOverride))))))

(fn CombineModifiers
  (params (ItemOwner :type Actor%20Object%20Reference) (AccumulatedModifiers :type Creative%20Melee%20Modifier%20Options%20Structure%20%28by%20ref%29))
  (bind (_damageplayeradd_21_494fd97541d4f0520f854fb319b2e853 _damageplayermult_22_f168716742252fc272e9aaae92d38cee _damageaiadd_23_b7d849c743710fa3e555a18c0d32ffbe _damageaimult_24_bd1f0cb248c75414f1906fac3726a0c8 _damageenvrionmentadd_25_f16ebbf64c5446d214efee9f5c84683d _damageenvrionmentmult_26_4882de4945cc65d5995fe58d85b2e1b1 _knockbackplayeradd_31_46cfdb8f4941e642043b2983d5c385f5 _knockbackaiadd_32_d8dc593a4b40a28c7277c9a947b7220b _comboenergycostadd_132_c87ec0354eb67cc449cb25813480bea8) (Utilities|Struct|BreakCreativeMeleeModifierOptions (LocalVariables|GetmyModifiers)))
  (bind (_hit1damadd_58_305564fb4873dff883e80598cad9ef9d _hit1dammult_59_5e3f7f094edfee10970af39e8186c9e5 _hit1knockbackmult_60_4b0da90d469255f43b98be9f953e9312 _hit1energymult_61_e2bec0ea4b2bff5999ee499a787daf86 _hit2damadd_76_3e8601914eb673f7bfa14d92d25dbd42 _hit2dammult_80_a6b0788c45b20718cdee6b8ff8e6f279 _hit2knockbackmult_84_e4d9e2f942e3777f3e3518b2c859f4bb _hit2energymult_88_f50a473a4891e16c3e09428a441c9f2e _hit3damadd_77_9e4afa9f4797f737716de1b224a1bd7a _hit3dammult_81_142df65d4afe4c8420ed6782050b8962 _hit3knockbackmult_85_9a37f8824390ee071d69a78c5b633f39 _hit3energymult_89_3052a23b4fd0a43830996bbd4804590f _hit4damadd_78_69d6ad414297f32770bf6db788555d1b _hit4dammult_82_6f0d71a549338d54760aea88ed93d8e7 _hit4knockbackmult_86_8b11253d42d410e0782494a1997e52b9 _hit4energymult_90_f7f44a9d491ffe77e4835ab1518cdc32 _hit5damadd_79_4b925546445816c98d443386f429b36d _hit5dammult_83_3bde1c574ba8d655a1849cbb98d261e8 _hit5knockbackmult_87_22b183714c4e6c0653b1c4a070dce2fb _hit5energymult_91_df0909af4a1f9c0fc02173b98271f83d) (Utilities|Struct|BreakCreativeMeleeModifierOptions (LocalVariables|GetmyModifiers)))
  (bind (_damageadd_2_7cec514c49df151a9d453491b01e4f93 _damagemult_20_200dd6cd436b4c20577174ac49310020 _critchance_27_1b71a1494c54ac4c7ac7d29caff00aa4 _critdamage_28_66fe207a47f411267a118fa55b30d580 _knockbackadd_29_dda919654e1b8959a586b7a358284ce3 _knockbackmult_30_e1d934af4c63176e2f905cbefd4d56aa _energycostadd_105_326d529b46e858d45447c99058290b4d _energycostmult_106_42c56f8943098c5e7126fa8a1ff3c567 _movementbonus_107_a6536df041780566a4b9bca2fe3d1d2b) (Utilities|Struct|BreakCreativeMeleeModifierOptions (LocalVariables|GetmyModifiers)))
  (bind (_chargeattackdamageadd_52_3176a2b14f91c9b93557ad9fdb6c70e3 _chargeattackdamagemult_53_435d848547cbd97806a7f981762cb4a9 _chargeattackknockback_54_142d815243467a83e732089c62daba27 _chargeattackcost_55_96697e5b4a48cf875d2ea08c1b1c5553 _chargetimeadd_56_18fa8ab14ea7f17605b1a694ec189e6e _chargetimemult_57_ad55b88343ccc3719ebb66b9420742e0) (Utilities|Struct|BreakCreativeMeleeModifierOptions (LocalVariables|GetmyModifiers)))
  (bind (_blockresistadd_101_465abb514c1610b643f8ffb81354e78b _blockresistmult_102_168b5d64491f00141ba2be943655deb6 _blockcostadd_128_0e7a7cdd469ea615ec2863910cb9d664 _blockcostmult_129_afdbf000445b18a6309c148379dcd41a _dodgecostadd_103_bc2ced9743d7fcda19c3449327b41883 _dodgecostmult_104_b13cd3214bc139f9b64610906e25a657) (Utilities|Struct|BreakCreativeMeleeModifierOptions (LocalVariables|GetmyModifiers)))
  (bind (_jumpattackdamageadd_108_74fc14134de8d68c8796f6a2ac9a0f40 _jumpattackdamagemult_109_e14a07824facae9812220da7dd827275 _jumpattackknockback_110_c62835bd45a7b5ea56e62b970de4b240 _jumpattackcost_111_1ff24cc649aefc9c0427088db2eaa980) (Utilities|Struct|BreakCreativeMeleeModifierOptions (LocalVariables|GetmyModifiers)))
  (bind (_sprintattackdamageadd_112_f2f36e9444d861b3bd3b85a4965e9113 _sprintattackdamagemult_113_9743c66b438df4377b1aa48a895fc2a4 _sprintattackknockback_117_1271f5294dbf26c9dfd55299ddcda8f5 _sprintattackcost_118_0d2b902b423227dc68ea20aa53b074ce) (Utilities|Struct|BreakCreativeMeleeModifierOptions (LocalVariables|GetmyModifiers)))
  (bind (_colorintensitycustom_122_dbc0ae3c46c05c8bbd194bb6f5b27f2d _colorintensityweapon_121_00ed43c2488af7d4c35e0da305ed7695) (Utilities|Struct|BreakCreativeMeleeModifierOptions (LocalVariables|GetmyModifiers)))
  (bind _valid (ValidforPlayer|ValidForActor ItemOwner))
  (if _valid
    (LocalVariables|SetmyModifiers AccumulatedModifiers)
    (Utilities|FlowControl|Sequence
      (:then_0
        ;[1 1 1] Weapon Core
        (if (Utilities|Enum|Equal(Enum) (Variables|UserOptions|GetModifyWeaponCore) "NewEnumerator2")
          (Utilities|Struct|SetmembersinCreativeMeleeModifierOptions (LocalVariables|GetmyModifiers) (+ _damageadd_2_7cec514c49df151a9d453491b01e4f93 (Variables|UserOptions|WeaponCore|GetmDamageAdd)) (+ _damagemult_20_200dd6cd436b4c20577174ac49310020 (Variables|UserOptions|WeaponCore|GetmDamageMult)) (+ _critchance_27_1b71a1494c54ac4c7ac7d29caff00aa4 (Variables|UserOptions|WeaponCore|GetmCritChance)) (+ _critdamage_28_66fe207a47f411267a118fa55b30d580 (Variables|UserOptions|WeaponCore|GetmCritDamage)) (+ _knockbackadd_29_dda919654e1b8959a586b7a358284ce3 (Variables|UserOptions|WeaponCore|GetmKnockbackAdd)) (+ _knockbackmult_30_e1d934af4c63176e2f905cbefd4d56aa (Variables|UserOptions|WeaponCore|GetmKnockbackMult)) (+ _energycostadd_105_326d529b46e858d45447c99058290b4d (Variables|UserOptions|Extra|GetmEnergyCostAdd)) (+ _energycostmult_106_42c56f8943098c5e7126fa8a1ff3c567 (Variables|UserOptions|Extra|GetmEnergyCostMult)) (+ _movementbonus_107_a6536df041780566a4b9bca2fe3d1d2b (Variables|UserOptions|Extra|GetmMovementBonus)))
          (else
            _))
        ;[1 1 1] Basic Attack
        (if (Utilities|Enum|Equal(Enum) (Variables|UserOptions|Damage|GetModifyDamage) "NewEnumerator2")
          (Utilities|Struct|SetmembersinCreativeMeleeModifierOptions (LocalVariables|GetmyModifiers) (+ _damageplayeradd_21_494fd97541d4f0520f854fb319b2e853 (Variables|UserOptions|Damage|GetmDamagePlayerAdd)) (+ _damageplayermult_22_f168716742252fc272e9aaae92d38cee (Variables|UserOptions|Damage|GetmDamagePlayerMult)) (+ _damageaiadd_23_b7d849c743710fa3e555a18c0d32ffbe (Variables|UserOptions|Damage|GetmDamageAIAdd)) (+ _damageaimult_24_bd1f0cb248c75414f1906fac3726a0c8 (Variables|UserOptions|Damage|GetmDamageAIMult)) (+ _damageenvrionmentadd_25_f16ebbf64c5446d214efee9f5c84683d (Variables|UserOptions|Damage|GetmDamageEnvironmentAdd)) (+ _damageenvrionmentmult_26_4882de4945cc65d5995fe58d85b2e1b1 (Variables|UserOptions|Damage|GetmDamageEnvironmentMult)) (+ _knockbackplayeradd_31_46cfdb8f4941e642043b2983d5c385f5 (Variables|UserOptions|Damage|GetmKnockbackPlayerAdd)) (+ _knockbackaiadd_32_d8dc593a4b40a28c7277c9a947b7220b (Variables|UserOptions|Damage|GetmKnockbackAIAdd)) (+ _comboenergycostadd_132_c87ec0354eb67cc449cb25813480bea8 (Variables|UserOptions|Extra|GetmComboEnergyCostAdd)))
          (else
            _))
        ;[1 1 1] Charge Attack
        (if (Utilities|Enum|Equal(Enum) (Variables|UserOptions|ChargeAttack|GetModifyChargeAttack) "NewEnumerator2")
          (Utilities|Struct|SetmembersinCreativeMeleeModifierOptions (LocalVariables|GetmyModifiers) (+ _chargeattackdamageadd_52_3176a2b14f91c9b93557ad9fdb6c70e3 (Variables|UserOptions|ChargeAttack|GetmChargeAttackDamageAdd)) (+ _chargeattackdamagemult_53_435d848547cbd97806a7f981762cb4a9 (Variables|UserOptions|ChargeAttack|GetmChargeAttackDamageMult)) (+ _chargeattackknockback_54_142d815243467a83e732089c62daba27 (Variables|UserOptions|ChargeAttack|GetmChargeAttackKnockback)) (+ _chargeattackcost_55_96697e5b4a48cf875d2ea08c1b1c5553 (Variables|UserOptions|ChargeAttack|GetmChargeAttackCost)) (+ _chargetimeadd_56_18fa8ab14ea7f17605b1a694ec189e6e (Variables|UserOptions|ChargeAttack|GetmChargeTimeAdd)) (+ _chargetimemult_57_ad55b88343ccc3719ebb66b9420742e0 (Variables|UserOptions|ChargeAttack|GetmChargeTimeMult)))
          (else
            _))
        ;[1 1 1] Secondary
        (if (Utilities|Enum|Equal(Enum) (Variables|UserOptions|Extra|GetModifySecondaryAction) "NewEnumerator2")
          (Utilities|Struct|SetmembersinCreativeMeleeModifierOptions (LocalVariables|GetmyModifiers) (+ _blockresistadd_101_465abb514c1610b643f8ffb81354e78b (Variables|UserOptions|Extra|GetmBlockResistAdd)) (+ _blockresistmult_102_168b5d64491f00141ba2be943655deb6 (Variables|UserOptions|Extra|GetmBlockResistMult)) (+ _blockcostadd_128_0e7a7cdd469ea615ec2863910cb9d664 (Variables|UserOptions|Extra|GetmBlockCostAdd)) (+ _blockcostmult_129_afdbf000445b18a6309c148379dcd41a (Variables|UserOptions|Extra|GetmBlockCostMult)) (+ _dodgecostadd_103_bc2ced9743d7fcda19c3449327b41883 (Variables|UserOptions|Extra|GetmDodgeCostAdd)) (+ _dodgecostmult_104_b13cd3214bc139f9b64610906e25a657 (Variables|UserOptions|Extra|GetmDodgeCostMult)))
          (else
            _))
        ;[1 1 1] Jump Attack
        (if (Utilities|Enum|Equal(Enum) (Variables|UserOptions|JumpAttack|GetModifyJumpAttack) "NewEnumerator2")
          (Utilities|Struct|SetmembersinCreativeMeleeModifierOptions (LocalVariables|GetmyModifiers) (+ _jumpattackdamageadd_108_74fc14134de8d68c8796f6a2ac9a0f40 (Variables|UserOptions|JumpAttack|GetmJumpAttackDamageAdd)) (+ _jumpattackdamagemult_109_e14a07824facae9812220da7dd827275 (Variables|UserOptions|JumpAttack|GetmJumpAttackDamageMult)) (+ _jumpattackknockback_110_c62835bd45a7b5ea56e62b970de4b240 (Variables|UserOptions|JumpAttack|GetmJumpAttackKnockback)) (+ _jumpattackcost_111_1ff24cc649aefc9c0427088db2eaa980 (Variables|UserOptions|JumpAttack|GetmJumpAttackCost)))
          (else
            _))
        ;[1 1 1] Sprint Attack
        (if (Utilities|Enum|Equal(Enum) (Variables|UserOptions|SprintAttack|GetModifySprintAttack) "NewEnumerator2")
          (Utilities|Struct|SetmembersinCreativeMeleeModifierOptions (LocalVariables|GetmyModifiers) (+ _sprintattackdamageadd_112_f2f36e9444d861b3bd3b85a4965e9113 (Variables|UserOptions|SprintAttack|GetmSprintAttackDamageAdd)) (+ _sprintattackdamagemult_113_9743c66b438df4377b1aa48a895fc2a4 (Variables|UserOptions|SprintAttack|GetmSprintAttackDamageMult)) (+ _sprintattackknockback_117_1271f5294dbf26c9dfd55299ddcda8f5 (Variables|UserOptions|SprintAttack|GetmSprintAttackKnockback)) (+ _sprintattackcost_118_0d2b902b423227dc68ea20aa53b074ce (Variables|UserOptions|SprintAttack|GetmSprintAttackCostAdd)))
          (else
            _))
        ;[1 1 1] Presentation
        (if (Utilities|Enum|Equal(Enum) (Variables|UserOptions|Presentation|GetModifyPresentation) "NewEnumerator2")
          (Utilities|Struct|SetmembersinCreativeMeleeModifierOptions (LocalVariables|GetmyModifiers) (+ _colorintensitycustom_122_dbc0ae3c46c05c8bbd194bb6f5b27f2d (Variables|UserOptions|Presentation|GetmColorIntensityCustom)) (+ _colorintensityweapon_121_00ed43c2488af7d4c35e0da305ed7695 (Variables|UserOptions|Presentation|GetmColorIntensityWeapon)))
          (else
            _))
        ;[1 1 1] Attack Combo
        (if (Utilities|Enum|Equal(Enum) (Variables|UserOptions|Attacks|GetModifyCombo) "NewEnumerator2")
          (Utilities|Struct|SetmembersinCreativeMeleeModifierOptions (LocalVariables|GetmyModifiers) (+ _hit1damadd_58_305564fb4873dff883e80598cad9ef9d (Variables|UserOptions|Attacks|Attack1|GetmHit1DamAdd)) (+ _hit1dammult_59_5e3f7f094edfee10970af39e8186c9e5 (Variables|UserOptions|Attacks|Attack1|GetmHit1DamMult)) (+ _hit1knockbackmult_60_4b0da90d469255f43b98be9f953e9312 (Variables|UserOptions|Attacks|Attack1|GetmHit1KnockMult)) (+ _hit1energymult_61_e2bec0ea4b2bff5999ee499a787daf86 (Variables|UserOptions|Attacks|Attack1|GetmHit1EnergyMult)) (+ _hit2damadd_76_3e8601914eb673f7bfa14d92d25dbd42 (Variables|UserOptions|Attacks|Attack2|GetmHit2DamAdd)) (+ _hit2dammult_80_a6b0788c45b20718cdee6b8ff8e6f279 (Variables|UserOptions|Attacks|Attack2|GetmHit2DamMult)) (+ _hit2knockbackmult_84_e4d9e2f942e3777f3e3518b2c859f4bb (Variables|UserOptions|Attacks|Attack2|GetmHit2KnockMult)) (+ _hit2energymult_88_f50a473a4891e16c3e09428a441c9f2e (Variables|UserOptions|Attacks|Attack2|GetmHit2EnergyMult)) (+ _hit3damadd_77_9e4afa9f4797f737716de1b224a1bd7a (Variables|UserOptions|Attacks|Attack3|GetmHit3DamAdd)) (+ _hit3dammult_81_142df65d4afe4c8420ed6782050b8962 (Variables|UserOptions|Attacks|Attack3|GetmHit3DamMult)) (+ _hit3knockbackmult_85_9a37f8824390ee071d69a78c5b633f39 (Variables|UserOptions|Attacks|Attack3|GetmHit3KnockMult)) (+ _hit3energymult_89_3052a23b4fd0a43830996bbd4804590f (Variables|UserOptions|Attacks|Attack3|GetmHit3EnergyMult)) (+ _hit4damadd_78_69d6ad414297f32770bf6db788555d1b (Variables|UserOptions|Attacks|Attack4|GetmHit4DamAdd)) (+ _hit4dammult_82_6f0d71a549338d54760aea88ed93d8e7 (Variables|UserOptions|Attacks|Attack4|GetmHit4DamMult)) (+ _hit4knockbackmult_86_8b11253d42d410e0782494a1997e52b9 (Variables|UserOptions|Attacks|Attack4|GetmHit4KnockMult)) (+ _hit4energymult_90_f7f44a9d491ffe77e4835ab1518cdc32 (Variables|UserOptions|Attacks|Attack4|GetmHit4EnergyMult)) (+ _hit5damadd_79_4b925546445816c98d443386f429b36d (Variables|UserOptions|Attacks|Attack5|GetmHit5DamAdd)) (+ _hit5dammult_83_3bde1c574ba8d655a1849cbb98d261e8 (Variables|UserOptions|Attacks|Attack5|GetmHit5DamMult)) (+ _hit5knockbackmult_87_22b183714c4e6c0653b1c4a070dce2fb (Variables|UserOptions|Attacks|Attack5|GetmHit5KnockMult)) (+ _hit5energymult_91_df0909af4a1f9c0fc02173b98271f83d (Variables|UserOptions|Attacks|Attack5|GetmHit5EnergyMult)))))
      (:then_1
        (Variables|SetBy-RefVar AccumulatedModifiers (LocalVariables|GetmyModifiers))))))

(fn GetCannonballImpactRadius
  (bind _cannonballimpactradius (Variables|UserOptions|ChargeAttack|CannonBall|GetCannonballImpactRadius))
  (if (Math|Float|NearlyEqual(Float) _cannonballimpactradius 0.0 0.01)
    (return 256.0)
    (else
      (return _cannonballimpactradius))))

(fn GetComboModValues
  (params (HitTag :type Gameplay%20Tag%20Structure) (CombinedModifiers :type Creative%20Melee%20Modifier%20Options%20Structure%20%28by%20ref%29))
  (bind (_hit5damadd_79_4b925546445816c98d443386f429b36d _hit5dammult_83_3bde1c574ba8d655a1849cbb98d261e8 _hit5knockbackmult_87_22b183714c4e6c0653b1c4a070dce2fb _hit5energymult_91_df0909af4a1f9c0fc02173b98271f83d) (Utilities|Struct|BreakCreativeMeleeModifierOptions (Variables|GetCombinedModifiers)))
  (bind (_hit1damadd_58_305564fb4873dff883e80598cad9ef9d _hit1dammult_59_5e3f7f094edfee10970af39e8186c9e5 _hit1knockbackmult_60_4b0da90d469255f43b98be9f953e9312 _hit1energymult_61_e2bec0ea4b2bff5999ee499a787daf86) (Utilities|Struct|BreakCreativeMeleeModifierOptions (Variables|GetCombinedModifiers)))
  (bind (_hit2damadd_76_3e8601914eb673f7bfa14d92d25dbd42 _hit2dammult_80_a6b0788c45b20718cdee6b8ff8e6f279 _hit2knockbackmult_84_e4d9e2f942e3777f3e3518b2c859f4bb _hit2energymult_88_f50a473a4891e16c3e09428a441c9f2e) (Utilities|Struct|BreakCreativeMeleeModifierOptions (Variables|GetCombinedModifiers)))
  (bind (_hit3damadd_77_9e4afa9f4797f737716de1b224a1bd7a _hit3dammult_81_142df65d4afe4c8420ed6782050b8962 _hit3knockbackmult_85_9a37f8824390ee071d69a78c5b633f39 _hit3energymult_89_3052a23b4fd0a43830996bbd4804590f) (Utilities|Struct|BreakCreativeMeleeModifierOptions (Variables|GetCombinedModifiers)))
  (bind (_hit4damadd_78_69d6ad414297f32770bf6db788555d1b _hit4dammult_82_6f0d71a549338d54760aea88ed93d8e7 _hit4knockbackmult_86_8b11253d42d410e0782494a1997e52b9 _hit4energymult_90_f7f44a9d491ffe77e4835ab1518cdc32) (Utilities|Struct|BreakCreativeMeleeModifierOptions (Variables|GetCombinedModifiers)))
  (if (GameplayTags|Equal(GameplayTag) HitTag "(TagName=\"Gameplay.Melee.Hit1\")")
    (return _hit1damadd_58_305564fb4873dff883e80598cad9ef9d _hit1dammult_59_5e3f7f094edfee10970af39e8186c9e5 _hit1knockbackmult_60_4b0da90d469255f43b98be9f953e9312 _hit1energymult_61_e2bec0ea4b2bff5999ee499a787daf86)
    (elif (GameplayTags|Equal(GameplayTag) HitTag "(TagName=\"Gameplay.Melee.Hit2\")")
      (return _hit2damadd_76_3e8601914eb673f7bfa14d92d25dbd42 _hit2dammult_80_a6b0788c45b20718cdee6b8ff8e6f279 _hit2knockbackmult_84_e4d9e2f942e3777f3e3518b2c859f4bb _hit2energymult_88_f50a473a4891e16c3e09428a441c9f2e)
      (elif (GameplayTags|Equal(GameplayTag) HitTag "(TagName=\"Gameplay.Melee.Hit3\")")
        (return _hit3damadd_77_9e4afa9f4797f737716de1b224a1bd7a _hit3dammult_81_142df65d4afe4c8420ed6782050b8962 _hit3knockbackmult_85_9a37f8824390ee071d69a78c5b633f39 _hit3energymult_89_3052a23b4fd0a43830996bbd4804590f)
        (elif (GameplayTags|Equal(GameplayTag) HitTag "(TagName=\"Gameplay.Melee.Hit4\")")
          (return _hit4damadd_78_69d6ad414297f32770bf6db788555d1b _hit4dammult_82_6f0d71a549338d54760aea88ed93d8e7 _hit4knockbackmult_86_8b11253d42d410e0782494a1997e52b9 _hit4energymult_90_f7f44a9d491ffe77e4835ab1518cdc32)
          (elif (GameplayTags|Equal(GameplayTag) HitTag "(TagName=\"Gameplay.Melee.Hit5\")")
            (return _hit5damadd_79_4b925546445816c98d443386f429b36d _hit5dammult_83_3bde1c574ba8d655a1849cbb98d261e8 _hit5knockbackmult_87_22b183714c4e6c0653b1c4a070dce2fb _hit5energymult_91_df0909af4a1f9c0fc02173b98271f83d)
            (else
              (return 0.0 0.0 0.0 0.0))))))))

(fn GetComboValues
  (params (HitTag :type Gameplay%20Tag%20Structure))
  (if (GameplayTags|Equal(GameplayTag) HitTag "(TagName=\"Gameplay.Melee.Hit1\")")
    (return (Variables|UserOptions|Attacks|Attack1|GetHit1DamMult) (Variables|UserOptions|Attacks|Attack1|GetHit1KnockMult) (Variables|UserOptions|Attacks|Attack1|GetHit1EnergyMult))
    (elif (GameplayTags|Equal(GameplayTag) HitTag "(TagName=\"Gameplay.Melee.Hit2\")")
      (return (Variables|UserOptions|Attacks|Attack2|GetHit2DamMult) (Variables|UserOptions|Attacks|Attack2|GetHit2KnockMult) (Variables|UserOptions|Attacks|Attack2|GetHit2EnergyMult))
      (elif (GameplayTags|Equal(GameplayTag) HitTag "(TagName=\"Gameplay.Melee.Hit3\")")
        (return (Variables|UserOptions|Attacks|Attack3|GetHit3DamMult) (Variables|UserOptions|Attacks|Attack3|GetHit3KnockMult) (Variables|UserOptions|Attacks|Attack3|GetHit3EnergyMult))
        (elif (GameplayTags|Equal(GameplayTag) HitTag "(TagName=\"Gameplay.Melee.Hit4\")")
          (return (Variables|UserOptions|Attacks|Attack4|GetHit4DamMult) (Variables|UserOptions|Attacks|Attack4|GetHit4KnockMult) (Variables|UserOptions|Attacks|Attack4|GetHit4EnergyMult))
          (elif (GameplayTags|Equal(GameplayTag) HitTag "(TagName=\"Gameplay.Melee.Hit5\")")
            (return (Variables|UserOptions|Attacks|Attack5|GetHit5DamMult) (Variables|UserOptions|Attacks|Attack5|GetHit5KnockMult) (Variables|UserOptions|Attacks|Attack5|GetHit5EnergyMult))
            (else
              (return 0.0 0.0 0.0))))))))

(fn IsValidTeam
  (params (Actor :type Actor%20Object%20Reference))
  ;[1 1 1] Any Team
  (if (== (Math|Byte|MakeLiteralByte (Creative|TeamOption|ConverttoTeam (Variables|UserOptions|Restrictions|GetTeamIndex))) 0)
    (return (not (Variables|UserOptions|Restrictions|GetTeamInvert)))
    (else
      ;[1 1 1] Specific Team
      (return (select (Variables|UserOptions|Restrictions|GetTeamInvert) (not (== (Team|GetActorTeam (Variables|GetActor)) (Math|Byte|MakeLiteralByte (Creative|TeamOption|ConverttoTeam (Variables|UserOptions|Restrictions|GetTeamIndex))))) (== (Team|GetActorTeam (Variables|GetActor)) (Math|Byte|MakeLiteralByte (Creative|TeamOption|ConverttoTeam (Variables|UserOptions|Restrictions|GetTeamIndex)))))))))

(fn IsManager
  (return (GameplayTags|HasMatchingGameplayTag "(TagName=\"Creative.Device.Item.Designer\")")))

(fn IsModifier
  (if (GameplayTags|HasMatchingGameplayTag "(TagName=\"Creative.Device.Item.Modifier\")")
    (return true (Variables|UserOptions|GetmBehavior))
    (else
      (return false "NewEnumerator5"))))

(fn IsValidClass
  (params (Actor :type Actor%20Object%20Reference))
  (bind _actor (Variables|GetActor))
  (bind _returnvalue (Minigame|GetMinigame _actor :!class /Script/FortniteGame.MinigameBlueprintLibrary))
  (bind _returnvalue_2 (Minigame|GetFNEClassComponent _returnvalue))
  (bind _actor_2 (Variables|GetActor))
  (bind _returnvalue_3 (Fortnite|GetFortPlayerControllerfromActor _actor_2))
  ;[1 1 1] Fallthrough...
  (Utilities|FlowControl|Sequence
    (:then_0
      ;[1 1 1] Any Class
      (if (Creative|ClassOption|IsClassType (Variables|UserOptions|Restrictions|GetClassIndex) "Any")
        (return (not (Variables|UserOptions|Restrictions|GetClassInvert)))
        (else
          ;[1 1 1] Specific Class
          (Utilities|IsValid _returnvalue_3
            (:Is%20Valid
              (bind _playerstate (Variables|/Script/Engine.Controller|GetPlayerState _returnvalue_3)
                (:then
                  (bind _asfort_player_state (Utilities|Casting|CastToFortPlayerState _playerstate))
                  (Utilities|IsValid _returnvalue
                    (:Is%20Valid
                      (Utilities|IsValid _returnvalue_2
                        (:Is%20Valid
                          (return (!= (Variables|UserOptions|Restrictions|GetClassInvert) (== (Minigame|GetClassSlotforPlayer _returnvalue_2 _asfort_player_state false) (Creative|ClassOption|ConverttoClassSlotIndex (Variables|UserOptions|Restrictions|GetClassIndex))))))))))
                (:else)))))))
    (:then_1
      (return false))))

(fn IsValidPlayer
  (params (Actor :type Actor%20Object%20Reference))
  ;[1 1 1] Fallthrough...
  (Utilities|FlowControl|Sequence
    (:then_0
      ;[1 1 1] 0 == Skip over tracked characters.
      (if (== (Variables|Default|GetRequireRegisteredPlayers) 0)
        (return true)
        (else
          ;[1 1 1] Is the character tracked?
          (bind _actor (Variables|GetActor)
            (:then
              (Utilities|IsValid (Fortnite|GetControllerfromActor _actor)
                (:Is%20Valid
                  (if (Utilities|Array|ContainsItem (Variables|Default|GetTrackedPlayers) (Fortnite|GetControllerfromActor _actor))
                    (return true)))))
            (:else)))))
    (:then_1
      (return false))))

(fn TrackPlayer
  (params (Add :type Boolean) (Controller :type Controller%20Object%20Reference))
  (bind _trackedplayers (Variables|Default|GetTrackedPlayers))
  (if Add
    (Utilities|Array|AddUnique _trackedplayers Controller)
    (else
      (Utilities|Array|RemoveItem _trackedplayers Controller))))

(fn ValidforActor
  (params (ItemOwner :type Actor%20Object%20Reference))
  (Utilities|IsValid ItemOwner
    (:Is%20Valid
      (return (Math|Boolean|ANDBoolean (ValidforPlayer|IsValidClass ItemOwner) (ValidforPlayer|IsValidTeam ItemOwner) (ValidforPlayer|IsValidPlayer ItemOwner))))
    (:Is%20Not%20Valid
      (return false))))
