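; Blueprint for a wrapped sledgehammer melee weapon, judging by its name, derived from a parent weapon class.
; Its GetSecondaryAbility override passes through the parent's ability tag, secondary damage, knockback and knockup unchanged.
; The construction script only calls the parent, and BeginPlay and Tick are disabled stubs, so this graph adds no logic of its own.
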(event EventBeginPlay
  (|CallParentFunction) ; This node is disabled and will not be called.\nDrag off pins to build functionality.
)

(event EventTick
  (params (DeltaSeconds :type Float%20%28single-precision%29))
  (|CallParentFunction DeltaSeconds) ; This node is disabled and will not be called.\nDrag off pins to build functionality.
)

(fn GetSecondaryAbility
  (bind (_abilitytag _secondary_damage _secondary_knockback _secondary_knockup) (|CallParentFunction))
  (return _abilitytag _secondary_damage _secondary_knockback _secondary_knockup))

(fn ConstructionScript
  (|CallParentFunction))
