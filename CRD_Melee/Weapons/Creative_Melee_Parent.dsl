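; Parent Blueprint for Creative melee weapons, resolving designer-tunable stats and serving them to combat abilities through an interface.
; Item variant changes rebuild the cached state from manager or data-table defaults, rank-ordered overrides, then combined modifiers.
; That state drives the async-loaded secondary attack, HUD hints, VFX color and stat abilities activated on equip.
; Replicated charge state, deliberately client-settable for prediction, triggers cosmetic charge and discharge VFX and audio.
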(event Custom|Set%20Charge%20State
  (params (New%20Charge%20State :type EMeleeChargeState%20Enum :default "NewEnumerator0"))
  ;[0 0.4982 1 0.25] Setter Event to Replicate Charge State. Deliberately not Authority Only so local client can predict changes.
  (Variables|Default|SetCharge%20State New%20Charge%20State))

(event FortWeapon|EventOnPlayImpactFX
  (params (HitResult :type Hit%20Result%20Structure%20%28by%20ref%29) (ImpactPhysicalSurface :type EPhysicalSurface%20Enum :default "SurfaceType_Default") (SpawnedPSC :type FXSystem%20Component%20Object%20Reference))
  ;[1 1 1] Impact VFX custom color
  (CallFunction|CheckAndSetCustomVFXColor SpawnedPSC)
  (|CallParentFunction HitResult ImpactPhysicalSurface SpawnedPSC))

(event OnItemTemplateEnabled(ItemVariantNotifier)
  (params (ItemTemplate :type Fort%20Creative%20Item%20Template%20Object%20Reference))
  (bind _asfort_player_controller (CallFunction|GetInstigatingController))
  (bind _returnvalue (Ability|GetFortAbilitySystemComponent _asfort_player_controller))
  ;[1 0.004949 0.9857] This is called whenever anything has changed that may adjust how this weapon performs.
  (ItemTemplates|WeaponDataUpdate)
  (Utilities|FlowControl|Sequence
    (:then_0
      (Equip|HandleWeaponUnequipped _returnvalue))
    (:then_1
      (Equip|HandleWeaponEquipped _returnvalue))))

(event Custom|CancelCharge%20FX
  ;[1 0.5615 0 0.25] Charge fx
  (Components|Activation|Deactivate (Variables|Default|GetChargeUP%20Instance))
  (Variables|Default|SetChargeUP%20Instance 0)
  (Variables|Default|SetSharedChargeReleased 0)
  (Audio|Components|Audio|FadeOut (Variables|Default|GetChargeLoopAudioComp) 0.2 0.0 "Linear" :!class /Script/Engine.AudioComponent))

(event OnItemTemplateDisabled(ItemVariantNotifier)
  (params (ItemTemplate :type Fort%20Creative%20Item%20Template%20Object%20Reference))
  (bind _asfort_player_controller (CallFunction|GetInstigatingController))
  (bind _returnvalue (Ability|GetFortAbilitySystemComponent _asfort_player_controller))
  ;[1 0.004949 0.9857] This is called whenever anything has changed that may adjust how this weapon performs.
  (ItemTemplates|WeaponDataUpdate)
  (Utilities|FlowControl|Sequence
    (:then_0
      (Equip|HandleWeaponUnequipped _returnvalue))
    (:then_1
      (Equip|HandleWeaponEquipped _returnvalue))))

(event Custom|PlayStopChargeSFX
  ;[1 0.5615 0 0.25] Charge fx
  (bind _audiocomp (CallFunction|PlayMeleeSoundLibAttached (Variables|Default|GetChargeReleasedSoundLibTag)))
  (Variables|Default|SetChargeReleasedAudioComp _audiocomp)
  (Audio|Components|Audio|FadeOut (Variables|Default|GetChargeLoopAudioComp) 0.6 0.0 "Linear" :!class /Script/Engine.AudioComponent))

(event OnItemTemplateChanged(ItemVariantNotifier)
  (params (ItemTemplate :type Fort%20Creative%20Item%20Template%20Object%20Reference))
  (bind _asfort_player_controller (CallFunction|GetInstigatingController))
  (bind _returnvalue (Ability|GetFortAbilitySystemComponent _asfort_player_controller))
  ;[1 0.004949 0.9857] This is called whenever anything has changed that may adjust how this weapon performs.
  (ItemTemplates|WeaponDataUpdate)
  (Utilities|FlowControl|Sequence
    (:then_0
      (Equip|HandleWeaponUnequipped _returnvalue))
    (:then_1
      (Equip|HandleWeaponEquipped _returnvalue))))

(event FortWeapon|EventOnEquippedAndAttached
  (|CallParentFunction)
  (ItemTemplate|BindtoItemVariant (Variables|Default|GetItemVariantNotifier) (CallFunction|GetItemVariantHandle))
  (ItemTemplates|WeaponDataUpdate)
  (Equip|HandleWeaponEquipped (Ability|GetFortAbilitySystemComponent (CallFunction|GetInstigatingController))))

(event Custom|Play%20Charged%20SFX
  ;[1 0.5615 0 0.25] Charge fx
  (bind _audiocomp (CallFunction|PlayMeleeSoundLibAttached (Variables|Default|GetChargeChargedSoundLibTag)))
  (Variables|Default|SetChargeChargedAudioComp _audiocomp))

(event FortWeapon|EventOnLocalUnEquip
  (|CallParentFunction)
  (ItemTemplate|UnbindfromItemVariant (Variables|Default|GetItemVariantNotifier) (CallFunction|GetItemVariantHandle))
  (ItemTemplates|ResetCachedItemTemplates)
  (Equip|HandleWeaponUnequipped (Ability|GetFortAbilitySystemComponent (CallFunction|GetInstigatingController))))

(event Custom|Play%20Charge%20VFX%20If%20Not%20Playing
  (bind (_total_charge_time _initial_charge_delay) (CallFunction|GetChargeDetails))
  ;[1 0.5615 0 0.25] Charge fx
  (Variables|Default|GetChargeUP%20Instance
    (:then)
    (:else
      (bind _object (Utilities|AsyncLoadAsset (CallFunction|FindChargeFX))
        (:Completed
          (if (or (Utilities|Enum|Equal(Enum) "NewEnumerator2" (Variables|Default|GetCharge%20State)) (Utilities|Enum|Equal(Enum) (Variables|Default|GetCharge%20State) "NewEnumerator3"))
            (bind _chargeup_instance (Variables|Default|GetChargeUP%20Instance)
              (:then
                (Components|Activation|Deactivate _chargeup_instance)
                (Variables|Default|SetChargeUP%20Instance 0))
              (:else
                _))
            (bind _asniagara_system (Utilities|Casting|CastToNiagaraSystem _object))
            (bind _returnvalue (Niagara|SpawnSystemAttached _asniagara_system (Variables|Visual|GetWeaponMesh) "None" "0, 0, 0" "0, 0, 0" "KeepRelativeOffset" true false))
            (bind _output_get (Variables|Default|SetChargeUP%20Instance _returnvalue))
            (CallFunction|CheckAndSetCustomVFXColor _output_get)
            (Effects|Components|ParticleSystem|SetFloatParameter _output_get "Duration" (Math|Float|SelectFloat (- _total_charge_time _initial_charge_delay) 0.1 (Utilities|Enum|Equal(Enum) (Variables|Default|GetCharge%20State) "NewEnumerator2")))
            (bind _enabled (CreativeMeleeWeaponInterface|GetSecondaryBreakBlockEnabled))
            (Effects|Components|ParticleSystem|SetFloatParameter (Variables|Default|GetChargeUP%20Instance) "Blockbreak" (Math|Conversions|ToFloat(Boolean) _enabled))
            (Components|Activation|Activate (Variables|Default|GetChargeUP%20Instance) true)
            (Variables|Default|GetChargeLoopAudioComp
              (:then)
              (:else
                (bind _audiocomp (CallFunction|PlayMeleeSoundLibAttached (Variables|Default|GetChargeLoopSoundLibTag)))
                (Variables|Default|SetChargeLoopAudioComp _audiocomp)))))))))

(event Custom|Update%20Cached%20Secondary%20Ability
  (bind _cachedstate (Variables|Runtime|GetCachedState))
  (bind _secondaryattack_40_ba7f00fc4a13ebc0912b909733aab0b7 (Utilities|Struct|BreakCreativeMeleeOptions _cachedstate))
  ;[0 1 0.9303 0.25] Uses the cached data for the weapon to load and save a class reference to the ability we'll cast when secondary attacking. Used by the primary fire ability when we charge long enough to trigger the secondary attack (via the Creative Melee Weapon Interface)
  (CallFunction|EvaluateDefaultSecondaryAbility)
  (if (Utilities|Enum|NotEqual(Enum) _secondaryattack_40_ba7f00fc4a13ebc0912b909733aab0b7 "NewEnumerator0")
    (bind _class (Utilities|AsyncLoadClassAsset (Utilities|Map|Find (Variables|Default|GetSecondaryAttacks) _secondaryattack_40_ba7f00fc4a13ebc0912b909733aab0b7))
      (:Completed
        (bind _asgameplay_ability (Utilities|Casting|CastToGameplayAbilityClass _class))
        (Variables|Default|SetCached%20Secondary%20Attack _asgameplay_ability)))
    (else
      (Variables|Default|SetCached%20Secondary%20Attack 0))))

(event Custom|PlayDischargeFX
  ;[1 0.5615 0 0.25] Charge fx
  (Components|Activation|Deactivate (Variables|Default|GetChargeUP%20Instance))
  (bind _audiocomp (CallFunction|PlayMeleeSoundLibAttached (Variables|Default|GetChargeReleasedSoundLibTag)))
  (bind _output_get (Variables|Default|SetChargeReleasedAudioComp _audiocomp))
  (Audio|Parameter|SetBooleanParameter _output_get "IsCharged" true :!class /Script/AudioExtensions.AudioParameterControllerInterface)
  (Audio|Components|Audio|FadeOut (Variables|Default|GetChargeLoopAudioComp) 0.6 0.0 "Linear" :!class /Script/Engine.AudioComponent)
  (bind _discharge (Variables|Default|GetDischarge)
    (:then
      (Components|Activation|Deactivate _discharge)
      (Variables|Default|SetDischarge 0))
    (:else
      _))
  (bind _returnvalue (Niagara|SpawnSystemAttached (Variables|Default|GetDischargeVFX) (Variables|Visual|GetWeaponMesh) "None" "0, 0, 0" "0, 0, 0" "KeepRelativeOffset" true))
  (bind _output_get_2 (Variables|Default|SetDischarge _returnvalue))
  (CallFunction|CheckAndSetCustomVFXColor _output_get_2)
  (bind _enabled (CreativeMeleeWeaponInterface|GetSecondaryBreakBlockEnabled))
  (Effects|Components|ParticleSystem|SetFloatParameter (Variables|Default|GetDischarge) "Blockbreak" (Math|Conversions|ToFloat(Boolean) _enabled)))

(fn ApplyGrantableAbilitiesToAbilityComponent
  (params (AbilityComponent :type Fort%20Ability%20System%20Component%20Object%20Reference) (AbilityClass :type Fort%20Gameplay%20Ability%20Class%20Reference) (AbilityID :type MeleeManagerMantisAbilityID%20Enum :default "NewEnumerator0") (TotalStateValue :type Creative%20Melee%20Options%20Structure))
  (bind (_basedamage_16_b01234df44393696f04c92b8d351e18e _critchance_18_b6bd1e7e4d026531de947fa85545974c _critmultiplier_44_3919d87c484c89957bee4b9f19883947 _movementspeed_30_9444e7814d756147dba496bb6ea702c5 _knockback_20_288c1de84fb83660a91876b19b39ee3b) (Utilities|Struct|BreakCreativeMeleeOptions TotalStateValue))
  (bind _returnvalue (Math|Float|Clamp(Float) _movementspeed_30_9444e7814d756147dba496bb6ea702c5 -5e+02 5e+02))
  (bind _output (Utilities|ClampMinFloat _basedamage_16_b01234df44393696f04c92b8d351e18e))
  (bind _output_2 (Utilities|ClampMinFloat _critmultiplier_44_3919d87c484c89957bee4b9f19883947))
  (bind _output_3 (Utilities|ClampMinFloat _critchance_18_b6bd1e7e4d026531de947fa85545974c))
  (bind _abilityid (Variables|GetAbilityID))
  (bind _returnvalue_2 (Utilities|Select :Index _abilityid :NewEnumerator0 _output :NewEnumerator1 _output_3 :NewEnumerator3 _output_3 :NewEnumerator4 _output_2 :NewEnumerator5 _knockback_20_288c1de84fb83660a91876b19b39ee3b :NewEnumerator6 _output :NewEnumerator7 _returnvalue))
  (if (!= _returnvalue_2 0.0)
    (CallFunction|GrantAbilitytoController AbilityComponent AbilityClass _returnvalue_2)))

(fn ApplyTemplateDefaults
  (params (WeaponHolder :type Fort%20Player%20Controller%20Object%20Reference) (WeaponTemplates :type Array%20of%20B%20Melee%20Manager%20Variant%20Template%20Parent%20Object%20References) (WeaponState :type Creative%20Melee%20Options%20Structure%20%28by%20ref%29))
  ;[1 1 1] Apply Default Values
  (for _array_element WeaponTemplates
    (LocalVariables|SetmyTemplateIncrement _array_element)
    ;;[1 1 1] Find the Melee Designer
    (bind _mytemplateincrement (LocalVariables|GetmyTemplateIncrement)
      (:then
        (bind _manager (Class|BMeleeManagerVariantTemplateParent|IsManager _mytemplateincrement))
        (if _manager
          (ModifierLogic|ApplyOverrides (LocalVariables|GetmyTemplateIncrement) (Variables|GetWeaponHolder) WeaponState true)
          (return)))
      (:else)))
  ;;[1 1 1] Fallthrough: No Melee Designer
  (bind _returnvalue (Utilities|GetDataTableRow "/CRD_Melee/CreativeMeleeWeaponBalance.CreativeMeleeWeaponBalance" (Variables|Default|GetWeapon%20Name))
    (:then
      (Variables|SetBy-RefVar WeaponState _returnvalue)
      (return))
    (:RowNotFound
      ;;[1 0.04693 0] Error: No Manager, No Default Weapon Data???
      (Development|PrintString (Utilities|String|Append "ERROR, ROW IN CREATIVEMELEEWEAPONBALANCE NOT FOUND FOR: " (Utilities|GetDisplayName self) " - Row name expected: " (Utilities|String|ToString(Name) (Variables|Default|GetWeapon%20Name))) true true "(R=1.000000,G=0.114046,B=0.000000,A=1.000000)" 2e+01)
      (return))))

(fn ApplyTemplateModifiers
  (params (WeaponHolder :type Fort%20Player%20Controller%20Object%20Reference) (WeaponTemplates :type Array%20of%20B%20Melee%20Manager%20Variant%20Template%20Parent%20Object%20References) (WeaponState :type Creative%20Melee%20Options%20Structure%20%28by%20ref%29))
  ;[1 1 1] Interate through Modifiers, to combine values.  Then, apply combined values.  (Value = Base * SUM(Multipliers) + SUM(Additions))
  ;;[1 1 1] Combine Modifiers
  (for _array_element WeaponTemplates
    (Utilities|IsValid _array_element
      (:Is%20Valid
        (bind _output_get (LocalVariables|SetmyTemplateIncrement _array_element))
        (bind (_modifier _type) (Class|BMeleeManagerVariantTemplateParent|IsModifier _output_get))
        (if _modifier
          (if (Utilities|Enum|Equal(Enum) _type "NewEnumerator2")
            (ModifierLogic|CombineModifiers (LocalVariables|GetmyTemplateIncrement) (Variables|GetWeaponHolder) (LocalVariables|GetmyModifiers)))))))
  ;;[1 1 1] Apply Combined Modifiers
  (ModifierLogic|ApplyModifiers (LocalVariables|GetmyTemplateIncrement) WeaponState (LocalVariables|GetmyModifiers))
  (return))

(fn ApplyTemplateOverrides
  (params (WeaponHolder :type Fort%20Player%20Controller%20Object%20Reference) (WeaponTemplates :type Array%20of%20B%20Melee%20Manager%20Variant%20Template%20Parent%20Object%20References) (WeaponState :type Creative%20Melee%20Options%20Structure%20%28by%20ref%29))
  ;[1 1 1] Only iterate through "Override" Modifiers
  (for _array_element WeaponTemplates
    (bind _output_get (LocalVariables|SetmyModifierCheck _array_element))
    (bind (_modifier _type) (Class|BMeleeManagerVariantTemplateParent|IsModifier _output_get))
    (if (and _modifier (Utilities|Enum|Equal(Enum) _type "NewEnumerator4"))
      ;;[1 1 1] Create a multi-dimensional array of templates with the outer index being the priority, and the inner array containing all templates at that rank.
      (bind _output_get_2 (LocalVariables|SetmyModifierRank (Variables|/CRD_Melee/Weapons/B_MeleeManager_Variant_Template_Parent.B_MeleeManager_Variant_Template_Parent_C|GetModifier%20Rank (LocalVariables|GetmyModifierCheck))))
      (bind _result (ArrayMacros|GetValidIndex (LocalVariables|GetmyTemplateOverrides) _output_get_2)
        (:Valid%20Index
          (bind _result_objects (Utilities|Struct|BreakFortObjectArray _result))
          (bind _output_get_3 (LocalVariables|SetmyModifiersAtRank _result_objects))
          (Utilities|Array|Add _output_get_3 (LocalVariables|GetmyModifierCheck))
          (Utilities|Array|SetArrayElem (LocalVariables|GetmyTemplateOverrides) (LocalVariables|GetmyModifierRank) (Utilities|Struct|MakeFortObjectArray (LocalVariables|GetmyModifiersAtRank)) false))
        (:Invalid%20Index
          ;;;[1 1 1] No template at this priority.  Create the inner array.
          (Utilities|Array|SetArrayElem (LocalVariables|GetmyTemplateOverrides) (LocalVariables|GetmyModifierRank) (Utilities|Struct|MakeFortObjectArray (Utilities|Array|MakeArray (LocalVariables|GetmyModifierCheck))) true)))))
  ;[1 1 1] Iterate through all rank-ordered templates and apply overrides in order (lowest to highest)
  (for _array_element_2 (LocalVariables|GetmyTemplateOverrides)
    (bind _array_element_objects (Utilities|Struct|BreakFortObjectArray _array_element_2))
    (for _array_element_3 _array_element_objects
      (bind _asb_melee_manager_variant_template_parent (Utilities|Casting|CastToB_MeleeManager_Variant_Template_Parent _array_element_3))
      (ModifierLogic|ApplyOverrides _asb_melee_manager_variant_template_parent (Variables|GetWeaponHolder) (Variables|GetWeaponState) false)))
  (return))

(fn CheckandSetCustomVFXColor
  (params (FX%20System%20Comp :type FXSystem%20Component%20Object%20Reference))
  (bind (_usecustomcolor _customcolor _customvfxintensity) (CreativeMeleeWeaponInterface|GetCustomVFXColor :!class /CRD_Melee/Creative_Melee_Weapon_Interface.Creative_Melee_Weapon_Interface_C))
  (Effects|Components|ParticleSystem|SetColorParameter FX%20System%20Comp "Color" (Math|Color|SelectColor _customcolor "(R=1.000000,G=1.000000,B=1.000000,A=1.000000)" _usecustomcolor))
  (Effects|Components|ParticleSystem|SetFloatParameter FX%20System%20Comp "UseCustomColor" (select _usecustomcolor 1.0 0.0))
  (Effects|Components|ParticleSystem|SetFloatParameter FX%20System%20Comp "CustomColorIntensity" (Math|Float|SelectFloat _customvfxintensity 0.0 _usecustomcolor)))

(fn EvaluateDefaultSecondaryAbility
  ;[1 0.9474 0] Secondary Attack
  (if (Utilities|Enum|Equal(Enum) (Utilities|Struct|BreakCreativeMeleeOptions (Variables|Runtime|GetCachedState)) "NewEnumerator8")
    (bind _returnvalue (Utilities|GetDataTableRow "/CRD_Melee/CreativeMeleeWeaponBalance.CreativeMeleeWeaponBalance" (Variables|Default|GetWeapon%20Name))
      (:then
        (Utilities|Struct|SetmembersinCreativeMeleeOptions (Variables|Runtime|GetCachedState) (Utilities|Struct|BreakCreativeMeleeOptions _returnvalue)))
      (:RowNotFound
        (Development|PrintString (Utilities|String|Append "ERROR POST LOOP, ROW IN CREATIVEMELEEWEAPONBALANCE NOT FOUND FOR" (Utilities|GetDisplayName self) " - Row name expected:" (Utilities|String|ToString(Name) (Variables|Default|GetWeapon%20Name))) true true "(R=1.000000,G=0.114046,B=0.000000,A=1.000000)" 2e+01)))))

(fn FindChargeFX
  (bind _vfx_dictionary (Variables|Default|GetVFX%20Dictionary))
  (bind _output_get (Variables|Default|SetGameplay%20Tags (GameplayTags|GetOwnedGameplayTags (Data|GetWeaponData) :!class /Script/GameplayTags.GameplayTagAssetInterface)))
  (bind _array_element (Utilities|Array|ForEachLoopwithBreak (GameplayTags|BreakGameplayTagContainer _output_get))
    (:LoopBody
      (if (.ReturnValue (Utilities|Map|Find _vfx_dictionary _array_element))
        (Variables|Default|SetFound (Utilities|Map|Find _vfx_dictionary _array_element))))
    (:Completed
      _))
  (return (Variables|Default|GetFound)))

(fn GetCanAttack
  (params (CurrentEnergy%25 :type Float%20%28double-precision%29))
  (bind (_primary_attack_cost _secondary_attack_cost _block_hit_cost _dodge_cost) (CreativeMeleeWeaponInterface|GetEnergyCosts true))
  (return (select (> (Utilities|Struct|BreakCreativeMeleeOptions (Variables|Runtime|GetCachedState)) 0) true (>= (Variables|GetCurrentEnergy%25) (/ _primary_attack_cost 1e+02))) (>= (Variables|GetCurrentEnergy%25) (/ _secondary_attack_cost 1e+02)) (>= (Variables|GetCurrentEnergy%25) (/ _block_hit_cost 1e+02)) (>= (Variables|GetCurrentEnergy%25) (/ _dodge_cost 1e+02))))

(fn GetHitData
  (params (Application%20Tag :type Gameplay%20Tag%20Structure))
  (bind (_basedamage_16_b01234df44393696f04c92b8d351e18e _baseaidamage_92_50312ec54dca1e86ae32918d5be84b5d _environmentdamage_28_0a4087964fb24d1fcb33038d472ce9cb _critchance_18_b6bd1e7e4d026531de947fa85545974c _critmultiplier_44_3919d87c484c89957bee4b9f19883947 _hitmultipliers_37_7079495847780de7215d1aa2cd31f255) (Utilities|Struct|BreakCreativeMeleeOptions (Variables|Runtime|GetCachedState)))
  (bind (_value _returnvalue) (Utilities|Map|Find _hitmultipliers_37_7079495847780de7215d1aa2cd31f255 Application%20Tag))
  (bind _value_damagemult_2_8c55deb74dfdbb3c90fcc79182014c94 (Utilities|Struct|BreakCreativeMeleeHitOptions _value))
  (Utilities|FlowControl|Sequence
    (:then_0
      ;[1 0.5856 0 0.25] Test if we have enough energy, or if we're on low energy (so should have reduced damage)
      (bind _energycomp (Variables|Default|GetEnergyComp)
        (:then
          _)
        (:else
          (bind _playercharacter (Variables|Runtime|GetPlayerCharacter)
            (:then
              (Variables|Default|SetEnergyComp (Actor|GetComponentbyClass _playercharacter "/TacticalSprintGame/Gameplay/TacticalSprintEnergyComponent.TacticalSprintEnergyComponent_C")))
            (:else))))
      (bind _primary_attack_cost (CreativeMeleeWeaponInterface|GetEnergyCosts false))
      (LocalVariables|SetEnergyDamageMult (Math|Float|SelectFloat 1.0 (Utilities|Struct|BreakCreativeMeleeOptions (Variables|Runtime|GetCachedState)) (Energy|HasSufficientEnergy _energycomp _primary_attack_cost))))
    (:then_1
      (return (Math|Float|float*float (Math|Float|SelectFloat _value_damagemult_2_8c55deb74dfdbb3c90fcc79182014c94 1.0 _returnvalue) _basedamage_16_b01234df44393696f04c92b8d351e18e (LocalVariables|GetEnergyDamageMult)) (Math|Float|float*float _environmentdamage_28_0a4087964fb24d1fcb33038d472ce9cb (Math|Float|SelectFloat _value_damagemult_2_8c55deb74dfdbb3c90fcc79182014c94 1.0 _returnvalue) (LocalVariables|GetEnergyDamageMult)) (Math|Float|float*float (Math|Float|SelectFloat _value_damagemult_2_8c55deb74dfdbb3c90fcc79182014c94 1.0 _returnvalue) _baseaidamage_92_50312ec54dca1e86ae32918d5be84b5d (LocalVariables|GetEnergyDamageMult)) (* (Math|Float|SelectFloat 0.0 _critchance_18_b6bd1e7e4d026531de947fa85545974c (!= (LocalVariables|GetEnergyDamageMult) 1.0)) 0.01) (Math|Float|Max(Float) (- _critmultiplier_44_3919d87c484c89957bee4b9f19883947 1.0))))))

(fn GetKnockbackData
  (params (JumpAttack :type Boolean) (TargetAI :type Boolean))
  (bind _cachedstate (Variables|Runtime|GetCachedState))
  (bind (_knockback_20_288c1de84fb83660a91876b19b39ee3b _aiknockback_86_3c11a6804d38dce230c48086a05545b1 _hitmultipliers_37_7079495847780de7215d1aa2cd31f255 _jumpattackknockbackpower_65_446e41444a48e9bef6552c85cab864cb) (Utilities|Struct|BreakCreativeMeleeOptions _cachedstate))
  (bind (_damagemult_2_8c55deb74dfdbb3c90fcc79182014c94 _knockbackmult_4_58eb11c94a828b839feb1ba3c32f0baa) (Utilities|Struct|BreakCreativeMeleeHitOptions (Utilities|Map|Find _hitmultipliers_37_7079495847780de7215d1aa2cd31f255 (Variables|Default|GetApplication%20Tag))))
  (if JumpAttack
    (LocalVariables|SetKnockbackPower _jumpattackknockbackpower_65_446e41444a48e9bef6552c85cab864cb)
    (else
      (LocalVariables|SetKnockbackPower (select TargetAI _aiknockback_86_3c11a6804d38dce230c48086a05545b1 _knockback_20_288c1de84fb83660a91876b19b39ee3b))))
  (return (> (LocalVariables|GetKnockbackPower) 0.0) (Math|Float|Clamp(Float) (* (LocalVariables|GetKnockbackPower) _knockbackmult_4_58eb11c94a828b839feb1ba3c32f0baa) 0.0 1.5e+04) (* (LocalVariables|GetKnockbackPower) (Math|Float|SelectFloat (Variables|DefaultData|GetKnockup%20Multiplier) 0.0 (> _knockbackmult_4_58eb11c94a828b839feb1ba3c32f0baa 0)))))

(fn GetSecondaryAbility
  (bind (_secondarydamage_32_eb6925f04b8c5ca1ea314abbc9c1b68f _secondaryknockback_47_20879ce6494d1742b4c73d8863396a13) (Utilities|Struct|BreakCreativeMeleeOptions (Variables|Runtime|GetCachedState)))
  (return (Variables|Default|GetCached%20Secondary%20Attack) _secondarydamage_32_eb6925f04b8c5ca1ea314abbc9c1b68f (Math|Float|Clamp(Float) _secondaryknockback_47_20879ce6494d1742b4c73d8863396a13 0.0 1.5e+04) (* (Math|Float|Clamp(Float) _secondaryknockback_47_20879ce6494d1742b4c73d8863396a13 0.0 1.5e+04) (Variables|DefaultData|GetKnockup%20Multiplier))))

(fn GetBlockData
  (return (Utilities|Struct|BreakCreativeMeleeOptions (Variables|Runtime|GetCachedState)) (Variables|Default|GetBlock%20Break%20Montage)))

(fn GetCannonballImpactRadius
  (return (Utilities|Struct|BreakCreativeMeleeOptions (Variables|Runtime|GetCachedState))))

(fn GetChargeDetails
  (bind (_totalchargetime_51_158e8c1d4ad3710b7babdfb2edab874f _initialchargedelay_53_99dc99e247202e2739f893849657e5e8) (Utilities|Struct|BreakCreativeMeleeOptions (Variables|Runtime|GetCachedState)))
  (return (Utilities|ClampMinFloat _totalchargetime_51_158e8c1d4ad3710b7babdfb2edab874f) (Utilities|ClampMinFloat _initialchargedelay_53_99dc99e247202e2739f893849657e5e8)))

(fn GetCustomVFXColor
  (bind (_usecustomvfxcolor_117_8f10c30b4921204b19e6b1986ac7a8dc _customvfxcolor_116_f512fe414884619e9b80588069dac041 _customvfxintensity_122_e62480714f5bf3dab0a41298a5e42a6e _weaponglowintensity_123_72d787f14737f1b1cb073aa54af7554c) (Utilities|Struct|BreakCreativeMeleeOptions (Variables|Runtime|GetCachedState)))
  (return _usecustomvfxcolor_117_8f10c30b4921204b19e6b1986ac7a8dc _customvfxcolor_116_f512fe414884619e9b80588069dac041 (Utilities|RemapFloat _customvfxintensity_122_e62480714f5bf3dab0a41298a5e42a6e "(X=0.000000,Y=100.000000)" "(X=0.000000,Y=1.000000)" true) (Utilities|RemapFloat _weaponglowintensity_123_72d787f14737f1b1cb073aa54af7554c "(X=0.000000,Y=100.000000)" "(X=0.000000,Y=1.000000)" true)))

(fn GetEnergyCosts
  (params (NextHit :type Boolean))
  (bind (_hitmultipliers_37_7079495847780de7215d1aa2cd31f255 _primaryenergycost_98_800ed153424618a176c2beafc878afbc _secondaryenergycost_100_1556aeee4d244b5f2bc7c0ba9ea6b0d4 _blockhitenergycost_102_7652559547687830a0a870a6c1c2e754 _dodgeenergycost_104_b87c80884c0ce9a1f652ddb29205a0c6 _sprintattackenergycost_126_012485474b335a7b14c68f80be65f4a8 _jumpattackenergycost_129_dea941b7458e0ad503a8e5b69cae6527) (Utilities|Struct|BreakCreativeMeleeOptions (Variables|Runtime|GetCachedState)))
  (bind _value (Utilities|Map|Find _hitmultipliers_37_7079495847780de7215d1aa2cd31f255 (Utilities|Map|Find (select NextHit (Variables|Default|GetAttackToNextHitMap) (Variables|Default|GetAttackToCurrentHitMap)) (Variables|Default|GetCurrentAttackName))))
  (bind (_value_damagemult_2_8c55deb74dfdbb3c90fcc79182014c94 _value_knockbackmult_4_58eb11c94a828b839feb1ba3c32f0baa _value_energymult_8_ba7e352d44e5e045ed0aa78b25a83d7a) (Utilities|Struct|BreakCreativeMeleeHitOptions _value))
  (return (* _value_energymult_8_ba7e352d44e5e045ed0aa78b25a83d7a _primaryenergycost_98_800ed153424618a176c2beafc878afbc) _secondaryenergycost_100_1556aeee4d244b5f2bc7c0ba9ea6b0d4 _blockhitenergycost_102_7652559547687830a0a870a6c1c2e754 _dodgeenergycost_104_b87c80884c0ce9a1f652ddb29205a0c6 _sprintattackenergycost_126_012485474b335a7b14c68f80be65f4a8 _jumpattackenergycost_129_dea941b7458e0ad503a8e5b69cae6527))

(fn GetInstigatingController
  (Utilities|FlowControl|Sequence
    (:then_0
      (bind _asfort_pawn (Utilities|Casting|CastToFortPawn (Game|GetInstigator)))
      (bind _asfort_player_controller (Utilities|Casting|CastToFortPlayerController (Pawn|GetController _asfort_pawn)))
      (LocalVariables|SetOutController _asfort_player_controller))
    (:then_1
      (return (LocalVariables|GetOutController)))))

(fn GetItemVariantHandle
  (Utilities|FlowControl|Sequence
    (:then_0
      (bind _output_get (LocalVariables|SetInstigatingController (CallFunction|GetInstigatingController)))
      (Utilities|IsValid _output_get
        (:Is%20Valid
          (bind _output_get_2 (LocalVariables|SetWeaponItem (Inventory|GetInventoryItemWithGuid _output_get (FortWeapon|GetInventoryGUID))))
          (Utilities|IsValid _output_get_2
            (:Is%20Valid
              (LocalVariables|SetOutHandle (Item|GetItemVariantHandle _output_get_2)))
            (:Is%20Not%20Valid
              _)))
        (:Is%20Not%20Valid
          _))
      (LocalVariables|SetOutHandle (ItemTemplate|GetReplicatedWeaponVariantHandle (Variables|Default|GetItemVariantNotifier))))
    (:then_1
      (return (LocalVariables|GetOutHandle)))))

(fn GetJumpAttackDamage
  (return (Utilities|Struct|BreakCreativeMeleeOptions (Variables|Runtime|GetCachedState))))

(fn GetJumpAttackEnabled
  (return (Utilities|Struct|BreakCreativeMeleeOptions (Variables|Runtime|GetCachedState))))

(fn GetJumpAttackKnockbackPower
  (return (Utilities|Struct|BreakCreativeMeleeOptions (Variables|Runtime|GetCachedState))))

(fn GetKnockbackOptions
  (return (Utilities|Struct|BreakCreativeMeleeOptions (Variables|Runtime|GetCachedState))))

(fn GetSecondaryBreakBlockEnabled
  (return (Math|Conversions|ToBoolean(Integer) (Utilities|Struct|BreakCreativeMeleeOptions (Variables|Runtime|GetCachedState)))))

(fn GetSecondaryFeedbackScale
  (return (Utilities|Struct|BreakCreativeMeleeOptions (Variables|Runtime|GetCachedState))))

(fn GetSprintAttackDamage
  (bind (_sprintattackdamage_110_3585d12b48db48a3b455de9fefaecf97 _sprintattackknockback_132_04a113cc4f5780ef7198fbab20c9d342) (Utilities|Struct|BreakCreativeMeleeOptions (Variables|Runtime|GetCachedState)))
  (return _sprintattackdamage_110_3585d12b48db48a3b455de9fefaecf97 _sprintattackknockback_132_04a113cc4f5780ef7198fbab20c9d342))

(fn GetSprintAttackEnabled
  (return (Utilities|Struct|BreakCreativeMeleeOptions (Variables|Runtime|GetCachedState))))

(fn GetTargetActionSelection
  (return (Utilities|Struct|BreakCreativeMeleeOptions (Variables|Runtime|GetCachedState))))

(fn GrantAbilityToController
  (params (AbilitySystem :type Fort%20Ability%20System%20Component%20Object%20Reference) (AbilityClass :type Fort%20Gameplay%20Ability%20Class%20Reference) (Magnitude :type Float%20%28double-precision%29))
  (Ability|Advanced|GrantandActivateAbilityonSelfwithParams AbilitySystem AbilityClass self (Utilities|Struct|MakeGameplayEventData "(TagName=\"\")" 0 0 0 0 "()" "(GameplayTags=)" "(GameplayTags=)" Magnitude)))

(fn HandleWeaponEquipped
  (params (AbilityComponent :type Fort%20Ability%20System%20Component%20Object%20Reference))
  (for _array_element (Variables|Internal|GetGrantableAbilityList)
    (Default|ApplyGrantableAbilitiestoAbilityComponent AbilityComponent (Utilities|Struct|BreakMeleeManagerMantisAbilityIDLink _array_element) (.AbilityID_8_172B992343C43A68680ACDAEFA19350C (Utilities|Struct|BreakMeleeManagerMantisAbilityIDLink _array_element)) (Variables|Runtime|GetCachedState))))

(fn HandleWeaponUnequipped
  (params (AbilityComponent :type Ability%20System%20Component%20Object%20Reference))
  (for _array_element (Variables|Internal|GetRemovedAbilityList)
    (bind _array_element_ability_2_af66ca4449d69e618e4fedae5bfa0199 (Utilities|Struct|BreakMeleeManagerMantisAbilityIDLink _array_element))
    (Ability|Advanced|GrantandActivateAbilityonSelfwithParams AbilityComponent _array_element_ability_2_af66ca4449d69e618e4fedae5bfa0199 self (Utilities|Struct|MakeGameplayEventData "(TagName=\"\")"))))

(fn Melee_Effect_Color
  (return "0, 0, 0"))

(fn OnRep_ChargeState
  ;[0 0.9784 1 0.25] When Charge State Changes, enable or disable various effects. If entering idle, what effects play is also determined by its previous state.
  (if (Utilities|Enum|NotEqual(Enum) (Variables|Default|GetPreviousChargeState) (Variables|Default|GetCharge%20State))
    (Utilities|FlowControl|Sequence
      (:then_0
        (Networking|CanExecuteCosmeticEvents
          (:True
            (switch Utilities|FlowControl|Switch|SwitchonEMeleeChargeState (Variables|Default|GetCharge%20State)
              (:NewEnumerator1
                (CallFunction|CancelChargeFX)
                (CallFunction|PlayDischargeFX)
                (switch Utilities|FlowControl|Switch|SwitchonEMeleeChargeState (Variables|Default|GetPreviousChargeState)
                  (:NewEnumerator2
                    (CallFunction|PlayStopChargeSFX))
                  (:NewEnumerator3
                    (CallFunction|PlayDischargeFX))))
              (:NewEnumerator2
                (CallFunction|PlayChargeVFXifNotPlaying))
              (:NewEnumerator3
                (CallFunction|PlayChargeVFXifNotPlaying)
                (CallFunction|PlayChargedSFX))))))
      (:then_1
        (Variables|Default|SetPreviousChargeState (Variables|Default|GetCharge%20State))))))

(fn PlayMeleeSoundLibAttached
  (params (SoundLibTag :type Gameplay%20Tag%20Structure))
  (bind _asfort_pawn (Utilities|Casting|CastToFortPawn (Game|GetInstigator)))
  (bind _returnvalue (SoundLibrary|PlaySoundSimple (WorldSubsystems|GetSoundLibrarySubsystem) self (Utilities|Struct|MakeSimpleSoundSettings SoundLibTag true true true true 1.0 1.0 0.0 0 0 "None" (Variables|/Script/Engine.Actor|GetRootComponent _asfort_pawn)) :!class /Script/SoundLibrary.SoundLibrarySubsystem))
  (for (_array_element _array_index) (.AudioComponents (Utilities|Struct|BreakSoundLibraryPlayResult _returnvalue))
    (if (== _array_index 0)
      (Audio|Components|Audio|Play _array_element 0.0)
      (else
        (Audio|Components|Audio|Stop _array_element)
        (Utilities|Array|RemoveItem (.AudioComponents (Utilities|Struct|BreakSoundLibraryPlayResult _returnvalue)) _array_element))))
  (bind _result (ArrayMacros|SafeArrayGet (.AudioComponents (Utilities|Struct|BreakSoundLibraryPlayResult _returnvalue)))
    (:%20%20
      (return _result))))

(fn ResetCachedItemTemplates
  (Utilities|Array|Clear (Variables|Runtime|GetEnabledItemTemplates)))

(fn SetCurrentAttack
  (params (CurrentAttack :type Name))
  ;[0 0.6394 1] A Setter Called Externally, this sets the current attack so other objects can compare against it (primarily the PreventMovementAction AnimNotify)
  (Variables|Default|SetCurrentAttackName CurrentAttack))

(fn SetHitDataApplicationTag
  (params (Application%20Tag :type Gameplay%20Tag%20Structure))
  (Variables|Default|SetApplication%20Tag Application%20Tag))

(fn ToggleEvadeHUDMessaging
  (params (bVisible :type Boolean))
  (if (Utilities|Select :Index (Variables|Default|GetTarget%20Action%20Selection) :NewEnumerator0 false :NewEnumerator1 true :NewEnumerator2 false :NewEnumerator3 false)
    (WeaponHUDData|SetHudKeyActionVisibility 4 bVisible)
    (WeaponHUDData|SetHudKeyActionVisibility 3 (not bVisible))))

(fn UpdateHUDKeyDetails
  (bind _cachedstate (Variables|Runtime|GetCachedState))
  (bind (_secondaryattack_40_ba7f00fc4a13ebc0912b909733aab0b7 _targetactionselection_80_830d772e40ca5d733fc80dbfcd605bc9 _lockon_89_d7a1968e484421556a3a1f8484f7bd0b) (Utilities|Struct|BreakCreativeMeleeOptions _cachedstate))
  (bind _secondaryattacknames (Variables|Default|GetSecondaryAttackNames))
  (bind (_value _returnvalue) (Utilities|Map|Find _secondaryattacknames _secondaryattack_40_ba7f00fc4a13ebc0912b909733aab0b7))
  ;[0 1 0.9301] Update Secondary Attack Details
  (Variables|Default|SetLockOnSelection _lockon_89_d7a1968e484421556a3a1f8484f7bd0b)
  (Variables|Default|SetTarget%20Action%20Selection _targetactionselection_80_830d772e40ca5d733fc80dbfcd605bc9)
  (WeaponHUDData|SetHudKeyActionVisibility 1 _returnvalue)
  (WeaponHUDData|SetHudKeyActionDescription 1 (Utilities|Text|FormatText "NSLOCTEXT(\"[74E79C5243A0A4057395B7B8F345EA37]\", \"C3464ADF4D7B571792816FBC86BE5EE9\", \"(Hold) {secondaryattack}\")" _value))
  (WeaponHUDData|SetHudKeyActionVisibility 3 (Utilities|Select :Index (Variables|Default|GetTarget%20Action%20Selection) :NewEnumerator0 false :NewEnumerator1 true :NewEnumerator2 true :NewEnumerator3 true))
  (WeaponHUDData|SetHudKeyActionDescription 3 (Utilities|Select :Index (Variables|Default|GetTarget%20Action%20Selection) :NewEnumerator0 (Variables|Default|GetBlockText) :NewEnumerator1 (Variables|Default|GetBlockText) :NewEnumerator2 (Variables|Default|GetDodgeText) :NewEnumerator3 (Variables|Default|GetBlockText)))
  (WeaponHUDData|SetHudKeyActionVisibility 5 (Utilities|Enum|Equal(Enum) (Variables|Default|GetLockOnSelection) "NewEnumerator0")))

(fn UpdateWeaponVFX
  (bind (_usecustomcolor _customcolor _customvfxintensity _weaponglowintensity) (CreativeMeleeWeaponInterface|GetCustomVFXColor :!class /CRD_Melee/Creative_Melee_Weapon_Interface.Creative_Melee_Weapon_Interface_C))
  (for _array_element (Utilities|Array|MakeArray (Variables|Default|GetIdleFX%20Niagara) (Variables|Default|GetSwingFX%20Niagara) (Variables|Default|GetAnimTrail%20Niagara))
    (Effects|Components|ParticleSystem|SetColorParameter _array_element "Color" (Math|Color|SelectColor _customcolor "(R=1.000000,G=1.000000,B=1.000000,A=1.000000)" _usecustomcolor))
    (Effects|Components|ParticleSystem|SetFloatParameter _array_element "UseCustomColor" (select _usecustomcolor 1.0 0.0))
    (Effects|Components|ParticleSystem|SetFloatParameter _array_element "CustomColorIntensity" (Math|Float|SelectFloat _customvfxintensity 0.0 _usecustomcolor)))
  (Effects|Components|ParticleSystem|SetFloatParameter (Variables|Default|GetIdleFX%20Niagara) "IdleIntensity" _weaponglowintensity))

(fn ConstructionScript
  (|CallParentFunction))

(fn WeaponDataUpdate
  (LocalVariables|SetmyWeaponHolder (CallFunction|GetInstigatingController))
  ;[1 1 1] Accumulate Weapon Data from a Base, applying Overrides, and finally Modifiers.
  (Variables|Runtime|SetEnabledItemTemplates (ItemTemplate|GetTemplateObjects (Variables|Default|GetItemVariantNotifier) "/CRD_Melee/Weapons/B_MeleeManager_Variant_Template_Parent.B_MeleeManager_Variant_Template_Parent_C" true))
  (Variables|Runtime|SetCachedState (Utilities|GetBlankWeaponValues)) ; Clear Cached Stated
  (WeaponModifier|ApplyTemplateDefaults (LocalVariables|GetmyWeaponHolder) (Variables|Runtime|GetEnabledItemTemplates) (Variables|Runtime|GetCachedState))
  (WeaponModifier|ApplyTemplateOverrides (LocalVariables|GetmyWeaponHolder) (Variables|Runtime|GetEnabledItemTemplates) (Variables|Runtime|GetCachedState))
  (WeaponModifier|ApplyTemplateModifiers (LocalVariables|GetmyWeaponHolder) (Variables|Runtime|GetEnabledItemTemplates) (Variables|Runtime|GetCachedState))
  ;[1 1 1] Update Weapon, based on Accumulated Weapon Data
  (CallFunction|UpdateCachedSecondaryAbility)
  (CallFunction|UpdateHUDKeyDetails)
  (CallFunction|UpdateWeaponVFX))
